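# Niagara beam effect for the Rift Device, stretching from the owner to User.EndPoint, with User.Opacity driving center width and arc spread.
# CenterBeam and three partitioned Electricity ribbons reposition particles along that line every update, adding a shared rotating wave, traveling width pulses and helical arc offsets.
# Sparks, electric sprites and orbiting mesh fragments dress the source.
# EndCone spawns delayed, expanding meshes just before the endpoint, aligned to the beam direction.
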NiagaraSystem /LimeRock/Assets/FX/RiftDevice/NS_LimeRock_RiftDevice_Beam
dependencies {
 usedDataInterfaces [
  &NiagaraDataInterfaceCurve
  &NiagaraDataInterfaceMeshRendererInfo
  &NiagaraDataInterfaceVectorCurve
  &NiagaraDataInterfaceColorCurve
 ]
 usedDynamicInputs [
  &RandomRangeLinearColor
  &SimulationPosition
  &UniformRangedFloat
  &Vector2DFromFloat
  &NormalizeVector
  &Subtract_Position
  /Niagara/DynamicInputs/Vectors/CrossProduct
 ]
 usedModules [
  &self:SystemSpawnScript.SystemScriptSource.SystemScriptGraph.NiagaraNodeAssignment_2.SetVariables_#1
  &SystemState
  &self:SystemSpawnScript.SystemScriptSource.SystemScriptGraph.NiagaraNodeAssignment_1.SetVariables_#2
  &self:SystemSpawnScript.SystemScriptSource.SystemScriptGraph.NiagaraNodeAssignment_0.SetVariables_#3
 ]
}
emitters {
 BaseSparks {
  emitterUpdateScript [
   &EmitterState(Life Cycle Mode=::NewEnumerator1, Loop Behavior=::NewEnumerator1, Loop Duration Mode=::NewEnumerator1, UseLoopDelay=false)
   &SpawnRate(Spawn Probability=0.6, SpawnRate=50, Use Spawn Probability=true)
   &SpawnBurst_Instantaneous(Spawn Count=18, Use Spawn Probability=false, UseLoopCountLimit=false)
   &SpawnBurst_Instantaneous(Spawn Count=9, Spawn Time=0.15, Use Spawn Probability=false, UseLoopCountLimit=false) {module=SpawnBurst_Instantaneous002}
   &SpawnBurst_Instantaneous(Spawn Count=55, Spawn Group=1, Spawn Time=0.6, Use Spawn Probability=false, UseLoopCountLimit=false) {module=SpawnBurst_Instantaneous001}
  ]
  particleSpawnScript [
   /SetVariables(Particles.Distribution:NiagaraFloat=&FloatFromCurve(
     FloatCurve=DataInterface() {propertyValues {Curve {keys [
         {interpMode=RCIM_Cubic, leaveTangent=0.4364, tangentMode=RCTM_Break, time=0, value=0}
         {arriveTangent=1.9884, interpMode=RCIM_Cubic, leaveTangent=1.9884, tangentMode=RCTM_User, time=0.64, value=0.5}
         <at 1 1 RCIM_Cubic>
        ]}}}
     CurveIndex=&RandomRangeFloat(Minimum=0, Maximum=1)
    )) {isSetParameters=true}
   &InitializeParticle(
    Color=&LerpLinearColors(StartColor=<rgba 0.215 0.4814 2.5 1>, EndColor=<rgba 0.0371 2.2165 3 1>, LerpFactor=$Particles.Distribution)
    Lifetime Min=0.4
    Lifetime Mode=::NewEnumerator1
    Sprite Size Mode=::NewEnumerator3
    Uniform Sprite Size=&Lerp_Float(A=1.5, B=3, Alpha=$Particles.Distribution)
    UsePositionOffset=false
    Write Material Random=true
   )
   &ShapeLocation(MaskSpawnGroup=false, Override Seed=false, Sphere Radius=3, Sphere Surface Distribution=1, UseOwnerScale=true)
   &AddVelocity(
    Cone Angle=100
    Cone Axis=$System.ForwardDirection
    Distribution Along Cone Axis=0.3
    Inner Cone Angle=20
    Module.UseConeFalloff=true
    Module.UseVelDistribution=true
    Override Seed=false
    Rotation Coordinate Space=::World
    Speed Falloff From Cone Axis=0.5
    Velocity Mode=::NewEnumerator2
    Velocity Speed=&Multiply_Float(
     A=&Lerp_Float(A=500, B=5000, Alpha=$Particles.Distribution)
     B=&MaskFloatBySpawnGroup(Masked Float=1, Passthrough Float=0.5, Spawn Group Mask=1)
    )
   )
  ]
  particleUpdateScript [
   &ParticleState()
   &DynamicMaterialParameters(
    Index 0 Param 1=&FloatFromCurve(FloatCurve=DataInterface() {propertyValues {Curve {keys [<at 0.1 0.5 RCIM_Cubic>,<at 1 4>]}}})
    Index 0 Param 2=5
    Index 0 Param 3=&FloatFromCurve(FloatCurve=DataInterface() {propertyValues {Curve {keys [<at 0 3 RCIM_Cubic>,<at 1 1>]}}})
    Index 0 Param 4=&FloatFromCurve(FloatCurve=DataInterface() {propertyValues {Curve {keys [{arriveTangent=-1.8572, interpMode=RCIM_Cubic, leaveTangent=-1.8572, tangentMode=RCTM_User, time=0, value=0.25}, <at 0.5 0>]}}})
    Module.Index0Float0Write=true
    Module.Index0Float1Write=false
    Module.Index0Float2Write=true
    Module.Index0Float3Write=true
   )
   &Drag(
    Drag=&FloatFromCurve(
     FloatCurve=DataInterface() {propertyValues {Curve {keys [
         {arriveTangent=10.6191, interpMode=RCIM_Cubic, leaveTangent=10.6191, tangentMode=RCTM_User, time=0, value=0.4}
         {arriveTangent=-0.8, interpMode=RCIM_Cubic, leaveTangent=-0.8, time=0.1, value=1}
         {arriveTangent=-3.1291, interpMode=RCIM_Cubic, leaveTangent=-3.1291, tangentMode=RCTM_User, time=0.5, value=0}
        ]}}}
     Scale Curve=4
    )
    Use Linear Drag=true
    Use Rotational Drag=false
   )
   &GravityForce(Gravity=<0 0 -1600>)
   &SolveForcesAndVelocity(Clamp Velocity=false, Limit Acceleration=false)
   &ScaleSpriteSize(
    Uniform Curve Scale=&FloatFromCurve(
     FloatCurve=DataInterface() {propertyValues {Curve {keys [
         {arriveTangent=0.001, interpMode=RCIM_Cubic, leaveTangent=0.001, tangentMode=RCTM_User, time=300, value=0.7}
         {arriveTangentWeight=166.6667, interpMode=RCIM_Cubic, leaveTangentWeight=400.0001, tangentMode=RCTM_User, time=800, value=1}
         <at 2000 2>
        ]}}}
     CurveIndex=$Engine.Owner.LODDistance
    )
    Uniform Curve Sprite Scale=DataInterface() {propertyValues {Curve {keys [<at 0.5 1 RCIM_Cubic>, {arriveTangent=-5.8778, interpMode=RCIM_Cubic, leaveTangent=-5.8778, tangentMode=RCTM_User, time=1, value=0}]}}}
   )
   &ScaleSpriteSizeBySpeed(
    Max Scale Factor=<0.8 8>
    Sample Scale Factor By Curve=true
    Scale Factor Curve=DataInterface() {propertyValues {Curve {keys [
        {arriveTangent=1.9498, interpMode=RCIM_Cubic, leaveTangent=1.9498, tangentMode=RCTM_User, time=0, value=0}
        {arriveTangent=2.6707, interpMode=RCIM_Cubic, leaveTangent=2.6707, tangentMode=RCTM_User, time=1, value=1}
       ]}}}
    Velocity Threshold=1800
   )
  ]
  rendererClasses [&NiagaraSpriteRendererProperties]
  renderers [{rendererClass=&NiagaraSpriteRendererProperties}]
 }
 CenterBeam {
  emitterUpdateScript [
   &EmitterState(Life Cycle Mode=::NewEnumerator1, Loop Behavior=::NewEnumerator1, Loop Duration Mode=::NewEnumerator1, UseLoopDelay=false)
   &SpawnBurst_Instantaneous(Spawn Count=20, Use Spawn Probability=false, UseLoopCountLimit=false)
  ]
  particleSpawnScript [&InitializeParticle(
    Color=<rgba 0.5 0.1302 0.4454 1>
    Lifetime=0.5
    Ribbon Width=&FloatFromCurve(
     FloatCurve=DataInterface() {propertyValues {Curve {keys [{arriveTangent=13.7494, interpMode=RCIM_Cubic, leaveTangent=13.7494, tangentMode=RCTM_User, time=0, value=0}, <at 0.15 1>]}}}
     CurveIndex=&ReturnNormalizedExecIndex()
     Scale Curve=80
    )
    Ribbon Width Mode=::NewEnumerator1
    UsePositionOffset=false
    Write Material Random=true
   )]
  particleUpdateScript [
   &ParticleState(Kill Particles When Lifetime Has Elapsed=false)
   /SetVariables(Particles.NormalizedDistance:NiagaraFloat=&NormalizeDistanceRange(
     Start Position:NiagaraPosition=$Engine.Owner.Position
     End Position:NiagaraPosition=$Particles.Position
     Distance:NiagaraFloat=&DistanceBetweenPositions(A=$Engine.Owner.Position, B=$User.EndPoint)
     Invert Normalized Range:NiagaraBool=false
    )) {isSetParameters=true}
   /SetVariables(Particles.Wavy:Vector3f=&RotateVectorByQuaternion(
     Vector To Rotate=&Multiply_VectorByFloat(
      Vector=$System.SideDirection
      Float=&FloatFromCurve(
       FloatCurve=DataInterface() {propertyValues {Curve {
          defaultValue=1
          keys [
           {arriveTangent=16.911, interpMode=RCIM_Cubic, leaveTangent=16.911, tangentMode=RCTM_User, time=0, value=0}
           {arriveTangentWeight=0.0333, interpMode=RCIM_Cubic, leaveTangentWeight=0.2333, tangentMode=RCTM_User, time=0.1, value=1}
           {arriveTangentWeight=0.3488, interpMode=RCIM_Cubic, leaveTangentWeight=0.0997, tangentMode=RCTM_User, time=0.8, value=1}
           {arriveTangent=-8.1388, tangentMode=RCTM_User, time=1, value=0.5}
          ]
         }}}
       CurveIndex=$Particles.NormalizedDistance
       Scale Curve=$System.Waviness
      )
     )
     Quaternion=&MakeQuaternion(
      Quaternion From=0
      Axis=$System.ForwardDirection
      Angle=&Multiply_Float(
       A=&Add_Float(A=$Particles.NormalizedDistance, B=&Multiply_Float(A=$Engine.System.Age, B=-0.5))
       B=2
      )
      From=?
      To=?
      Angle Type=::NewEnumerator1
      Normalize Vectors=false
     )
    )) {isSetParameters=true}
   /SetVariables(Particles.Position=&AddVectorToPosition(
     Position=&Lerp_Vector(
      A=Linked() {linkedVariable {name=Engine.Owner.Position, type=NiagaraPosition}}
      B=Linked() {linkedVariable {name=User.EndPoint, type=NiagaraPosition}}
      Alpha=&FloatFromCurve(
       FloatCurve=DataInterface() {propertyValues {Curve {defaultValue=1, keys [{tangentMode=RCTM_User, time=0, value=0}, <at 0.2 0.1>, <at 0.7 0.85>, {tangentMode=RCTM_User, time=1, value=1}]}}}
       CurveIndex=&ReturnNormalizedExecIndex()
       Scale Curve=&FloatFromCurve(FloatCurve=DataInterface() {propertyValues {Curve {
           defaultValue=1
           keys [
            {arriveTangent=3.0015, interpMode=RCIM_Cubic, time=0, value=0}
            {arriveTangent=1.9455, interpMode=RCIM_Cubic, leaveTangent=1.9455, tangentMode=RCTM_User, time=0.75, value=1}
           ]
          }}})
      )
     )
     Vector=$Particles.Wavy
    )) {isSetParameters=true}
   /SetVariables(Particles.Pulse:NiagaraFloat=&FloatFromCurve(
     FloatCurve=DataInterface() {propertyValues {Curve {
        defaultValue=1
        keys [
         {arriveTangent=10.3786, interpMode=RCIM_Cubic, time=0, value=0}
         {interpMode=RCIM_Cubic, tangentMode=RCTM_User, time=0.15, value=1}
         {arriveTangentWeight=0.1286, interpMode=RCIM_Cubic, leaveTangentWeight=0.2527, tangentMode=RCTM_User, time=0.3, value=0}
        ]
       }}}
     CurveIndex=&Add_Float(A=&Multiply_Float(A=$Particles.NormalizedDistance, B=-0.5), B=$Particles.NormalizedAge)
     Scale Curve=&FloatFromCurve(
      FloatCurve=DataInterface() {propertyValues {Curve {
         defaultValue=1
         keys [
          {arriveTangent=24.1854, interpMode=RCIM_Cubic, leaveTangent=24.1854, tangentMode=RCTM_User, time=0, value=0}
          {interpMode=RCIM_Cubic, tangentMode=RCTM_User, time=0.1, value=1}
          {arriveTangentWeight=0.3291, interpMode=RCIM_Cubic, leaveTangentWeight=0.1194, tangentMode=RCTM_User, time=0.7604, value=1}
          {arriveTangent=-3.1649, tangentMode=RCTM_User, time=1, value=0.5}
         ]
        }}}
      CurveIndex=$Particles.NormalizedDistance
     )
    )) {isSetParameters=true}
   &ScaleRibbonWidth(Ribbon Width Scale=&Multiply_Float(
     A=&Add_Float(
      A=&Lerp_Float(A=1, B=3, Alpha=$Particles.Pulse)
      B=&FloatFromCurve(
       FloatCurve=DataInterface() {propertyValues {Curve {keys [<at 0.5 0 RCIM_Cubic>, {arriveTangent=6.24, tangentMode=RCTM_User, time=1, value=1}]}}}
       CurveIndex=&ReturnNormalizedExecIndex()
       Scale Curve=&FloatFromCurve(
        FloatCurve=DataInterface() {propertyValues {Curve {
           defaultValue=1
           keys [{arriveTangent=10.4185, interpMode=RCIM_Cubic, leaveTangent=10.4185, tangentMode=RCTM_User, time=0.75, value=0}, <at 1 1 RCIM_Cubic>]
          }}}
        Scale Curve=7
       )
      )
     )
     B=$User.Opacity
    ))
  ]
  rendererClasses [&NiagaraSpriteRendererProperties, &NiagaraRibbonRendererProperties]
  renderers [
   {rendererClass=&NiagaraSpriteRendererProperties}
   {rendererClass=&NiagaraRibbonRendererProperties, rendererIndex=1}
  ]
 }
 ElectricSpirtes {
  emitterUpdateScript [
   &EmitterState(Life Cycle Mode=::NewEnumerator1, Loop Behavior=::NewEnumerator1, Loop Duration Mode=::NewEnumerator1, UseLoopDelay=false)
   &SpawnRate(Spawn Probability=0.6, SpawnGroup=1, SpawnRate=40, Use Spawn Probability=true)
   &SpawnBurst_Instantaneous(Spawn Count=3, Use Spawn Probability=false, UseLoopCountLimit=false)
   &SpawnBurst_Instantaneous(Spawn Count=5, Spawn Time=0.16, Use Spawn Probability=false, UseLoopCountLimit=false) {module=SpawnBurst_Instantaneous002}
   &SpawnBurst_Instantaneous(Spawn Count=7, Spawn Time=0.6, Use Spawn Probability=false, UseLoopCountLimit=false) {module=SpawnBurst_Instantaneous001}
  ]
  particleSpawnScript [
   &InitializeParticle(
    Color Maximum=<rgba 0.6028 4.5265 7 1>
    Color Minimum=<rgba 0.179 0.3128 3 1>
    Color Mode=::NewEnumerator2
    Lifetime Max=0.17
    Lifetime Min=0.12
    Lifetime Mode=::NewEnumerator1
    Mass Mode=::NewEnumerator1
    Sprite Rotation Mode=::NewEnumerator2
    Sprite Size=&UniformRangedVector2D(Minimum=<50 50>, Maximum=<200 200>)
    Sprite Size Mode=::NewEnumerator4
    UsePositionOffset=false
    Write Material Random=true
   ) {module=InitializeParticle001}
   &DynamicMaterialParameters(
    Index 0 Param 1=$Particles.MaterialRandom
    Index 0 Param 2=&Lerp_Float(A=0.7, B=1.25, Alpha=$Particles.MaterialRandom)
    Index 0 Param 4=1
    Module.Index0Float0Write=true
    Module.Index0Float1Write=true
    Module.Index0Float2Write=false
    Module.Index0Float3Write=true
   )
  ]
  particleUpdateScript [
   &ParticleState()
   &SpriteRotationRate(Rotation Rate=&RandomRangeFloat(Minimum=-100, Maximum=100))
   &DynamicMaterialParameters(
    Index 0 Param 3=&FloatFromCurve(FloatCurve=DataInterface() {propertyValues {Curve {keys [<at 0.5 0 RCIM_Cubic>, {arriveTangent=3.7979, interpMode=RCIM_Cubic, leaveTangent=3.7979, tangentMode=RCTM_User, time=1, value=1}]}}})
    Module.Index0Float0Write=false
    Module.Index0Float1Write=false
    Module.Index0Float2Write=true
    Module.Index0Float3Write=false
   ) {module=DynamicMaterialParameters001}
   &ScaleSpriteSize(
    Uniform Curve Scale=&MaskFloatBySpawnGroup(Masked Float=0.5, Passthrough Float=1, Spawn Group Mask=1)
    Uniform Curve Sprite Scale=DataInterface() {propertyValues {Curve {keys [{arriveTangent=0.4536, interpMode=RCIM_Cubic, leaveTangent=0.4536, tangentMode=RCTM_User, time=0, value=0.65}, <at 1 1>]}}}
   )
   &CameraOffset(Camera Offset Amount=/Niagara/DynamicInputs/Math/CalculateParticleRadius(
     Radius Calculation Type:ENiagaraCollisionRadiusOptions=::NewEnumerator0
     Particle Radius:NiagaraFloat=1
     Mesh Dimensions:Vector3f=<1 1 1>
     Method for Calculating Particle Radius:ENiagaraCalculateRadiusOptions=::NewEnumerator1
    ))
  ]
  rendererClasses [&NiagaraSpriteRendererProperties]
  renderers [{rendererClass=&NiagaraSpriteRendererProperties}]
 }
 Electricity {
  emitterUpdateScript [
   &EmitterState(Life Cycle Mode=::NewEnumerator1, Loop Behavior=::NewEnumerator1, Loop Duration Mode=::NewEnumerator1, UseLoopDelay=false)
   &SpawnBurst_Instantaneous(Spawn Count=45, Use Spawn Probability=false, UseLoopCountLimit=false)
  ]
  particleSpawnScript [
   &PartitionParticles(
    Normalized Index In Partition Includes One=true
    Number Of Partitions=3
    Particles Per Partition=15
    Partition Mode:ENiagara_PartitionMode_v2=::NewEnumerator1
    Write Partition Index to Ribbon Link Order:ENiagara_AttributeSamplingApplyOutput=::NewEnumerator0
    Write Partition to Ribbon ID=::NewEnumerator0
   )
   &InitializeParticle(
    Color=<rgba 0.0704 0.8173 1 1>
    Lifetime=0.7
    Ribbon Width=&FloatFromCurve(
     FloatCurve=DataInterface() {propertyValues {Curve {keys [<at 0 0.2>,<at 0.1 1>,<at 0.8 1>,<at 1 1.25>]}}}
     CurveIndex=$Particles.PartitionParticles.Normalized Index In Partition
     Scale Curve=20
    )
    Ribbon Width Mode=::NewEnumerator1
    Sprite Size Mode=::NewEnumerator3
    UsePositionOffset=false
    Write Material Random=true
   )
   /SetVariables(Particles.NormalizedIndexRemap:NiagaraFloat=&FloatFromCurve(
     FloatCurve=DataInterface() {propertyValues {Curve {defaultValue=1, keys [{tangentMode=RCTM_User, time=0, value=0}, <at 0.25 0.1>, <at 0.75 0.9>, {tangentMode=RCTM_User, time=1, value=1}]}}}
     CurveIndex=$Particles.PartitionParticles.Normalized Index In Partition
    )) {isSetParameters=true}
  ]
  particleUpdateScript [
   &ParticleState(Kill Particles When Lifetime Has Elapsed=false)
   /SetVariables(Particles.NormalizedDistance:NiagaraFloat=&NormalizeDistanceRange(
     Start Position:NiagaraPosition=$Engine.Owner.Position
     End Position:NiagaraPosition=$Particles.Position
     Distance:NiagaraFloat=&DistanceBetweenPositions(A=$Engine.Owner.Position, B=$User.EndPoint)
     Invert Normalized Range:NiagaraBool=false
    )) {isSetParameters=true}
   /SetVariables(Particles.Pulse:NiagaraFloat=&FloatFromCurve(
     FloatCurve=DataInterface() {propertyValues {Curve {
        defaultValue=1
        keys [
         {arriveTangent=10.3786, interpMode=RCIM_Cubic, leaveTangent=10.3786, tangentMode=RCTM_User, time=0, value=0}
         {interpMode=RCIM_Cubic, tangentMode=RCTM_User, time=0.2, value=1}
         {arriveTangentWeight=0.1286, interpMode=RCIM_Cubic, leaveTangentWeight=0.2527, tangentMode=RCTM_User, time=0.4, value=0}
        ]
       }}}
     CurveIndex=&Add_Float(
      A=&Multiply_Float(A=$Particles.PartitionParticles.Normalized Index In Partition, B=-0.5)
      B=$Particles.NormalizedAge
     )
     Scale Curve=&FloatFromCurve(
      FloatCurve=DataInterface() {propertyValues {Curve {
         defaultValue=1
         keys [
          {arriveTangent=23.8449, interpMode=RCIM_Cubic, leaveTangent=23.8449, tangentMode=RCTM_User, time=0, value=0}
          {interpMode=RCIM_Cubic, tangentMode=RCTM_User, time=0.1, value=1}
          {arriveTangentWeight=0.3291, interpMode=RCIM_Cubic, leaveTangentWeight=0.1194, tangentMode=RCTM_User, time=0.7604, value=1}
          {arriveTangent=-3.1649, tangentMode=RCTM_User, time=1, value=0.5}
         ]
        }}}
      CurveIndex=$Particles.PartitionParticles.Normalized Index In Partition
     )
    )) {isSetParameters=true}
   /SetVariables(Particles.Wavy:Vector3f=&RotateVectorByQuaternion(
     Vector To Rotate=&Multiply_VectorByFloat(
      Vector=$System.SideDirection
      Float=&FloatFromCurve(
       FloatCurve=DataInterface() {propertyValues {Curve {
          defaultValue=1
          keys [
           {arriveTangent=16.911, interpMode=RCIM_Cubic, leaveTangent=16.911, tangentMode=RCTM_User, time=0, value=0}
           {arriveTangentWeight=0.0333, interpMode=RCIM_Cubic, leaveTangentWeight=0.2333, tangentMode=RCTM_User, time=0.1, value=1}
          ]
         }}}
       CurveIndex=$Particles.NormalizedDistance
       Scale Curve=$System.Waviness
      )
     )
     Quaternion=&MakeQuaternion(
      Quaternion From=0
      Axis=$System.ForwardDirection
      Angle=&Multiply_Float(
       A=&Add_Float(A=$Particles.NormalizedDistance, B=&Multiply_Float(A=$Engine.System.Age, B=-0.5))
       B=2
      )
      From=?
      To=?
      Angle Type=::NewEnumerator1
      Normalize Vectors=false
     )
    )) {isSetParameters=true}
   /SetVariables(Particles.Position=&AddVectorToPosition(
     Position=&Lerp_Vector(
      A=Linked() {linkedVariable {name=Engine.Owner.Position, type=NiagaraPosition}}
      B=Linked() {linkedVariable {name=User.EndPoint, type=NiagaraPosition}}
      Alpha=$Particles.NormalizedIndexRemap
     )
     Vector=&Add_Vector(
      A=&RotateVectorByQuaternion(
       Vector To Rotate=&Multiply_VectorByFloat(
        Vector=$System.SideDirection
        Float=&Add_Float(
         A=&FloatFromCurve(
          FloatCurve=DataInterface() {propertyValues {Curve {keys [
              {arriveTangent=38.6814, interpMode=RCIM_Cubic, leaveTangent=38.6814, tangentMode=RCTM_User, time=0, value=0}
              {arriveTangentWeight=0.0174, interpMode=RCIM_Cubic, leaveTangentWeight=0.0242, tangentMode=RCTM_User, time=0.03, value=0.4}
              {arriveTangentWeight=0.0983, interpMode=RCIM_Cubic, leaveTangentWeight=0.843, tangentMode=RCTM_User, time=0.1, value=0.1}
              {arriveTangentWeight=0.2091, interpMode=RCIM_Cubic, leaveTangentWeight=0.1943, tangentMode=RCTM_User, time=0.7, value=0.15}
              {arriveTangent=5.5118, tangentMode=RCTM_User, time=1, value=1}
             ]}}}
          CurveIndex=$Particles.PartitionParticles.Normalized Index In Partition
          Scale Curve=&Multiply_Float(A=250, B=$User.Opacity)
         )
         B=&Multiply_Float(A=$Particles.Pulse, B=150)
        )
       )
       Quaternion=&MakeQuaternion(
        Quaternion From=0
        Axis=$System.ForwardDirection
        Angle=&Add_Float(
         A=&Multiply_FloatByInt(Float=0.3333, Integer=$Particles.PartitionParticles.Partition)
         B=&Add_Float(
          A=&Multiply_Float(A=$Particles.Age, B=-1)
          B=&Multiply_Float(A=$Particles.PartitionParticles.Normalized Index In Partition, B=3)
         )
        )
        From=?
        To=?
        Angle Type=::NewEnumerator1
        Normalize Vectors=false
       )
      )
      B=$Particles.Wavy
     )
    )) {isSetParameters=true}
   &self:ScratchModule_01(EndPos:NiagaraPosition=$User.EndPoint, StartPos:NiagaraPosition=$Engine.Owner.Position) {module=ScratchModule_01}
   &DynamicMaterialParameters(
    Index 0 Param 1=$Particles.ScratchModule_01.X
    Index 0 Param 2=$Particles.ScratchModule_01.Y
    Index 0 Param 3=$Particles.ScratchModule_01.Z
    Index 0 Param 4=&Add_Float(
     A=$Particles.ScratchModule_01.W
     B=&Multiply_FloatByInt(Float=0.5, Integer=$Particles.PartitionParticles.Partition)
    )
    Index 1 Param 2=&FloatFromCurve(FloatCurve=DataInterface() {propertyValues {Curve {defaultValue=1, keys [{arriveTangent=0.6031, interpMode=RCIM_Cubic, time=0, value=0.1}, <at 0.75 1 RCIM_Cubic>]}}})
    Index 1 Param 3=1
    Index 1 Param 4=&Lerp_Float(
     A=0.55
     B=&Add_Float(
      A=&FloatFromCurve(
       FloatCurve=DataInterface() {propertyValues {Curve {defaultValue=1, keys [<at 0.7 0 RCIM_Cubic>, {arriveTangent=3.8348, tangentMode=RCTM_User, time=1, value=0.45}]}}}
       CurveIndex=$Particles.PartitionParticles.Normalized Index In Partition
      )
      B=&Waveform(
       Global Phase=&Add_Float(A=&ReturnNormalizedExecIndex(Normalized Index Scale=1.5), B=$Engine.System.Age)
       "Waveform [1]"=::NewEnumerator0
       "[1] Phase Scale"=-1
       "[1] Frequency"=1
       "[1] Amplitude Min/Max"=<0 0.15>
      )
     )
     Alpha=$User.Opacity
    )
    Module.Index0Float0Write=true
    Module.Index0Float1Write=true
    Module.Index0Float2Write=true
    Module.Index0Float3Write=true
    Module.Index1Float0Write=false
    Module.Index1Float1Write=true
    Module.Index1Float2Write=true
    Module.Index1Float3Write=true
    Write Parameter Index 1=::NewEnumerator1
   )
   &ScaleColor(Scale RGB=&Lerp_Vector(A=<1 1 1>, B=<5 0.5 1>, Alpha=$Particles.Pulse), ScaleA=false, ScaleRGB=true)
   &ScaleRibbonWidth(Ribbon Width Scale=&Lerp_Float(A=1, B=4, Alpha=$Particles.Pulse))
   /SetVariables(
    Particles.Airglow_Color:LinearColor=&MultiplyLinearColor(Linear Color=$Particles.Color, Scale Factor=<1.1 0.9 1 0.25>)
    Particles.Airglow_Width:NiagaraFloat=&Multiply_Float(A=$Particles.RibbonWidth, B=3)
   ) {isSetParameters=true}
  ]
  rendererClasses [&NiagaraSpriteRendererProperties, &NiagaraRibbonRendererProperties]
  renderers [
   {rendererClass=&NiagaraSpriteRendererProperties}
   {rendererClass=&NiagaraRibbonRendererProperties, rendererIndex=1}
   {rendererClass=&NiagaraRibbonRendererProperties, rendererIndex=2}
  ]
 }
 EndCone {
  emitterUpdateScript [
   &EmitterState(Delay First Loop Only=true, Life Cycle Mode=::NewEnumerator1, Loop Delay=0.4, Loop Duration=1, UseLoopDelay=true)
   &SpawnRate(SpawnRate=5, Use Spawn Probability=false)
  ]
  particleSpawnScript [
   &InitializeParticle(
    Color=<rgba 3 3 3 1>
    Lifetime=0.6
    Mesh Scale=<6 3.5 3.5>
    Mesh Scale Mode=::NewEnumerator4
    UsePositionOffset=false
    Write Material Random=true
   )
   &CameraOffset(Camera Offset Amount=50)
   &InitialMeshOrientation(
    Mesh Orientation Mode=::NewEnumerator2
    Rotation=&RandomRangeVector(Minimum=<0 0 0>, Maximum=<1 0 0>)
    Use Rotation Vector=true
   )
  ]
  particleUpdateScript [
   &ParticleState()
   &UpdateMeshOrientation(
    Facing Direction=$System.ForwardDirection
    Facing Interpolation=false
    Orientation Method=::NewEnumerator1
    Override Quat=false
    TransformFacing=false
    TransformUp=false
   )
   /SetVariables(Particles.Position=&AddVectorToPosition(Position=$User.EndPoint, Vector=&Multiply_VectorByFloat(Vector=$System.ForwardDirection, Float=-150))) {isSetParameters=true}
   &ScaleMeshSize(Scale Factor=&VectorFromCurve(VectorCurve=DataInterface() {propertyValues {
       XCurve {
        defaultValue=1
        keys [
         {arriveTangent=2.0165, interpMode=RCIM_Cubic, leaveTangent=2.0165, tangentMode=RCTM_User, time=0, value=0}
         {arriveTangent=0.3302, tangentMode=RCTM_User, time=1, value=1}
        ]
       }
       YCurve {keys [
         {arriveTangent=1.4616, interpMode=RCIM_Cubic, leaveTangent=1.4616, tangentMode=RCTM_User, time=0, value=0.2}
         {
          arriveTangent=0.3173
          arriveTangentWeight=0.174
          interpMode=RCIM_Cubic
          leaveTangent=0.3173
          leaveTangentWeight=0.174
          tangentMode=RCTM_User
          time=1
          value=1
         }
        ]}
       ZCurve {keys [
         {arriveTangent=1.4616, interpMode=RCIM_Cubic, leaveTangent=1.4616, tangentMode=RCTM_User, time=0, value=0.2}
         {
          arriveTangent=0.3173
          arriveTangentWeight=0.174
          interpMode=RCIM_Cubic
          leaveTangent=0.3173
          leaveTangentWeight=0.174
          tangentMode=RCTM_User
          time=1
          value=1
         }
        ]}
      }}))
   &DynamicMaterialParameters(
    Index 0 Param 1=1.7
    Index 0 Param 2=-1
    Index 1 Param 1=&FloatFromCurve(FloatCurve=DataInterface() {propertyValues {Curve {
        defaultValue=1
        keys [
         {arriveTangent=3.2048, interpMode=RCIM_Cubic, leaveTangent=3.2048, tangentMode=RCTM_User, time=0, value=-1.5}
         {arriveTangent=1.589, tangentMode=RCTM_User, time=1, value=1}
        ]
       }}})
    Index 1 Param 2=&FloatFromCurve(
     FloatCurve=DataInterface() {propertyValues {Curve {
        defaultValue=1
        keys [
         {arriveTangent=3.1471, interpMode=RCIM_Cubic, leaveTangent=3.1471, tangentMode=RCTM_User, time=0, value=0}
         {tangentMode=RCTM_User, time=0.5, value=1}
        ]
       }}}
     Scale Curve=2
    )
    Module.Index0Float0Write=true
    Module.Index0Float1Write=true
    Module.Index0Float2Write=true
    Module.Index0Float3Write=true
    Module.Index1Float0Write=true
    Module.Index1Float1Write=true
    Module.Index1Float2Write=false
    Module.Index1Float3Write=false
    Write Parameter Index 1=::NewEnumerator1
   )
   &ScaleColor(
    Scale Alpha=&FloatFromCurve(FloatCurve=DataInterface() {propertyValues {Curve {
        defaultValue=1
        keys [
         {arriveTangent=17.5908, interpMode=RCIM_Cubic, time=-0.0041, value=0}
         {arriveTangent=-2.2176, interpMode=RCIM_Cubic, leaveTangent=-2.2176, tangentMode=RCTM_User, time=0.35, value=1}
         {tangentMode=RCTM_User, time=1, value=0}
        ]
       }}})
    Scale RGB=&MakeVectorFromLinearColorRGB(Color=&ColorFromCurve(ColorCurve=DataInterface() {propertyValues {
        AlphaCurve {keys [<at 0 1 RCIM_Cubic>]}
        BlueCurve {keys [<at 0.025 1 RCIM_Cubic>,<at 0.6054 1 RCIM_Cubic>]}
        GreenCurve {keys [<at 0.025 0.0497 RCIM_Cubic>,<at 0.6054 0.4754 RCIM_Cubic>]}
        RedCurve {keys [
          <at 0.025 0.5473 RCIM_Cubic>
          {arriveTangent=-4.8674, interpMode=RCIM_Cubic, leaveTangent=-4.8674, tangentMode=RCTM_User, time=0.287, value=0.5564}
          <at 0.6054 0.0375 RCIM_Cubic>
         ]}
       }}))
    ScaleA=true
    ScaleRGB=true
   )
  ]
  rendererClasses [&NiagaraSpriteRendererProperties, &NiagaraMeshRendererProperties]
  renderers [
   {rendererClass=&NiagaraSpriteRendererProperties}
   {rendererClass=&NiagaraMeshRendererProperties, rendererIndex=1}
  ]
 }
 Fragments {
  emitterUpdateScript [
   &EmitterState()
   &SpawnRate(SpawnRate=2, Use Spawn Probability=false)
  ]
  particleSpawnScript [
   &InitializeParticle(
    Color Maximum=<rgba 0.308 1.1825 2 1>
    Color Minimum=<rgba 0.1935 0.7371 1 1>
    Color Mode=::NewEnumerator2
    Lifetime Max=4.5
    Lifetime Min=3
    Lifetime Mode=::NewEnumerator1
    Mesh Renderer Array Visibility Mode=::NewEnumerator3
    Mesh Renderer Info=DataInterface() {propertyValues {MeshRenderer=&self:NiagaraEmitter_0.NiagaraMeshRendererProperties_0}}
    Mesh Scale Max=<0.7 0.7 0.7>
    Mesh Scale Min=<0.2 0.2 0.7>
    Mesh Scale Mode=::NewEnumerator2
    Sprite Size Mode=::NewEnumerator1
    Uniform Sprite Size Min=4
    UsePositionOffset=false
    Write Material Random=true
   )
   /SetVariables(Particles.OrbitSpeed:NiagaraFloat=&RandomRangeFloat(Minimum=0.05, Maximum=0.2)) {isSetParameters=true}
   &ShapeLocation(
    Apply To Particle Position=::NewEnumerator1
    MaskSpawnGroup=false
    Override Seed=false
    Rotation Coordinate Space=::World
    Sphere Radius=&RandomRangeFloat(Minimum=20, Maximum=35)
    UseOwnerScale=false
   )
   /SetVariables(Particles.MeshVis:NiagaraInt32=&RandomRangeInt(Minimum=0, Maximum=1)) {isSetParameters=true}
  ]
  particleUpdateScript [
   &ParticleState()
   &ScaleColor(
    Scale Alpha=&FloatFromCurve(FloatCurve=DataInterface() {propertyValues {Curve {
        defaultValue=0
        keys [
         {arriveTangent=4.5168, interpMode=RCIM_Cubic, leaveTangent=4.5168, tangentMode=RCTM_User, time=0, value=0}
         {arriveTangentWeight=0.1389, interpMode=RCIM_Cubic, leaveTangentWeight=0.2778, tangentMode=RCTM_User, time=0.25, value=1}
         {arriveTangentWeight=0.1667, interpMode=RCIM_Cubic, leaveTangentWeight=0.0833, tangentMode=RCTM_User, time=0.75, value=1}
         {arriveTangent=-5.3936, tangentMode=RCTM_User, time=1, value=0}
        ]
       }}})
    ScaleA=true
    ScaleRGB=false
   )
   &UpdateMeshOrientation(
    Rotation Rate=&RandomRangeFloat(Minimum=0.1, Maximum=0.4)
    Rotation Vector=&RandomVector(Vector Scale=1)
   )
   &ScaleSpriteSize(Uniform Curve Sprite Scale=DataInterface() {propertyValues {Curve {
       defaultValue=0
       keys [
        {arriveTangent=2.5266, interpMode=RCIM_Cubic, leaveTangent=2.5266, tangentMode=RCTM_User, time=0, value=0}
        {interpMode=RCIM_Cubic, tangentMode=RCTM_User, time=0.5, value=1}
        {arriveTangent=-3.9854, tangentMode=RCTM_User, time=1, value=0}
       ]
      }}})
   &InterpolateOverTime(
    Rate Of Change=7
    Target Position Value:NiagaraPosition=&AddVectorToPosition(
     Position=$Engine.Owner.Position
     Vector=&RotateVectorByQuaternion(
      Vector To Rotate=&RandomVector(Vector Scale=&RandomRangeFloat(Minimum=15, Maximum=25))
      Quaternion=&MakeQuaternion(
       Quaternion From=0
       Axis=&RandomVector(Vector Scale=1)
       Angle=&Multiply_Float(A=$Particles.Age, B=$Particles.OrbitSpeed)
       From=?
       To=?
       Angle Type=::NewEnumerator1
       Normalize Vectors=false
      )
     )
    )
    Variable Type=::NewEnumerator6
   )
   /SetVariables(Particles.Position=$Particles.InterpolateOverTime.InterpolateOverTime_Position.Moving Average) {isSetParameters=true}
  ]
  rendererClasses [&NiagaraSpriteRendererProperties, &NiagaraMeshRendererProperties]
  renderers [
   {rendererClass=&NiagaraSpriteRendererProperties}
   {rendererClass=&NiagaraMeshRendererProperties, rendererIndex=1}
  ]
 }
}
summary {
 systemName=NS_LimeRock_RiftDevice_Beam
 emitters [
  {emitterName=Electricity, rendererClasses [&NiagaraSpriteRendererProperties, &NiagaraRibbonRendererProperties]}
  {emitterName=CenterBeam, rendererClasses [&NiagaraSpriteRendererProperties, &NiagaraRibbonRendererProperties]}
  {emitterName=BaseSparks, rendererClasses [&NiagaraSpriteRendererProperties]}
  {emitterName=ElectricSpirtes, rendererClasses [&NiagaraSpriteRendererProperties]}
  {emitterName=Fragments, rendererClasses [&NiagaraSpriteRendererProperties, &NiagaraMeshRendererProperties]}
  {emitterName=EndCone, rendererClasses [&NiagaraSpriteRendererProperties, &NiagaraMeshRendererProperties]}
 ]
 userVariables [
  User.EndPoint {type=NiagaraPosition, description="", defaultValue {}}
  User.Opacity {type=NiagaraFloat, defaultValue=1, description=""}
 ]
}
