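# Niagara burst for the LimeRock rift device: sparks, electric sprites and light-emitting flares pulse three times, the last pulse strongest via spawn-group masks.
# One random distribution value sets spark color, size and speed; sparks fall with drag, stretch by speed and grow with LOD distance.
# Just before the last pulse, mesh and sprite fragments appear, smoothly chasing random points around the owner.
# System ColorShift tints all but the sparks blue to pink.
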NiagaraSystem /LimeRock/Assets/FX/RiftDevice/NS_LimeRock_RiftDevice_Burst
dependencies {
 usedDataInterfaces [
  &NiagaraDataInterfaceCurve
  &NiagaraDataInterfaceVector2DCurve
  &NiagaraDataInterfaceMeshRendererInfo
 ]
 usedDynamicInputs [
  &RandomRangeLinearColor
  &SimulationPosition
  &UniformRangedFloat
  &Vector2DFromFloat
 ]
 usedModules [
  &SystemState
  &self:SystemSpawnScript.SystemScriptSource.SystemScriptGraph.NiagaraNodeAssignment_0.SetVariables_#1
 ]
}
emitters {
 BaseSparks {
  emitterUpdateScript [
   &EmitterState(Life Cycle Mode=::NewEnumerator1, Loop Behavior=::NewEnumerator1, Loop Duration=1, UseLoopDelay=false)
   &SpawnRate(Spawn Probability=0.6, SpawnRate=25, Use Spawn Probability=true)
   &SpawnBurst_Instantaneous(Spawn Count=14, Use Spawn Probability=false, UseLoopCountLimit=false)
   &SpawnBurst_Instantaneous(Spawn Count=9, Spawn Time=0.3, Use Spawn Probability=false, UseLoopCountLimit=false) {module=SpawnBurst_Instantaneous002}
   &SpawnBurst_Instantaneous(Spawn Count=30, Spawn Group=1, Spawn Time=0.6, Use Spawn Probability=false, UseLoopCountLimit=false) {module=SpawnBurst_Instantaneous001}
  ]
  particleSpawnScript [
   /SetVariables(Particles.Distribution:NiagaraFloat=&FloatFromCurve(
     FloatCurve=DataInterface() {propertyValues {Curve {keys [
         {interpMode=RCIM_Cubic, leaveTangent=0.4364, tangentMode=RCTM_Break, time=0, value=0}
         {arriveTangent=1.9884, interpMode=RCIM_Cubic, leaveTangent=1.9884, tangentMode=RCTM_User, time=0.64, value=0.5}
         <at 1 1 RCIM_Cubic>
        ]}}}
     CurveIndex=&RandomRangeFloat(Minimum=0, Maximum=1)
    )) {isSetParameters=true}
   &InitializeParticle(
    Color=&LerpLinearColors(StartColor=<rgba 0.4975 0.1035 2.5 1>, EndColor=<rgba 3 0.1914 0.8276 1>, LerpFactor=$Particles.Distribution)
    Lifetime Min=0.3
    Lifetime Mode=::NewEnumerator1
    Sprite Size Mode=::NewEnumerator3
    Uniform Sprite Size=&Lerp_Float(A=1.5, B=3, Alpha=$Particles.Distribution)
    UsePositionOffset=false
    Write Material Random=true
   )
   &ShapeLocation(MaskSpawnGroup=false, Override Seed=false, Sphere Radius=3, Sphere Surface Distribution=1, UseOwnerScale=true)
   &AddVelocity(
    Module.Use Falloff Distance=false
    Origin Offset=<0 0 -1.5>
    Origin Offset Coordinate Space=::World
    Override Seed=false
    Velocity Mode=::NewEnumerator1
    Velocity Speed=&Multiply_Float(
     A=&Lerp_Float(A=150, B=1200, Alpha=$Particles.Distribution)
     B=&MaskFloatBySpawnGroup(Masked Float=1, Passthrough Float=0.6, Spawn Group Mask=1)
    )
   )
  ]
  particleUpdateScript [
   &ParticleState()
   &DynamicMaterialParameters(
    Index 0 Param 1=&FloatFromCurve(FloatCurve=DataInterface() {propertyValues {Curve {keys [<at 0.1 0.5 RCIM_Cubic>,<at 1 4>]}}})
    Index 0 Param 2=5
    Index 0 Param 3=&FloatFromCurve(FloatCurve=DataInterface() {propertyValues {Curve {keys [<at 0 3 RCIM_Cubic>,<at 1 1>]}}})
    Index 0 Param 4=&FloatFromCurve(FloatCurve=DataInterface() {propertyValues {Curve {keys [{arriveTangent=-1.8572, interpMode=RCIM_Cubic, leaveTangent=-1.8572, tangentMode=RCTM_User, time=0, value=0.25}, <at 0.5 0>]}}})
    Module.Index0Float0Write=true
    Module.Index0Float1Write=false
    Module.Index0Float2Write=true
    Module.Index0Float3Write=true
   )
   &Drag(
    Drag=&FloatFromCurve(
     FloatCurve=DataInterface() {propertyValues {Curve {keys [
         {arriveTangent=10.6191, interpMode=RCIM_Cubic, leaveTangent=10.6191, tangentMode=RCTM_User, time=0, value=0.4}
         {arriveTangent=-0.8, interpMode=RCIM_Cubic, leaveTangent=-0.8, time=0.1, value=1}
         {arriveTangent=-3.1291, interpMode=RCIM_Cubic, leaveTangent=-3.1291, tangentMode=RCTM_User, time=0.5, value=0}
        ]}}}
     Scale Curve=3
    )
    Use Linear Drag=true
    Use Rotational Drag=false
   )
   &GravityForce(Gravity=<0 0 -1600>)
   &SolveForcesAndVelocity(Clamp Velocity=false, Limit Acceleration=false)
   &ScaleSpriteSize(
    Uniform Curve Scale=&FloatFromCurve(
     FloatCurve=DataInterface() {propertyValues {Curve {keys [
         {arriveTangent=0.001, interpMode=RCIM_Cubic, leaveTangent=0.001, tangentMode=RCTM_User, time=300, value=0.7}
         {arriveTangentWeight=166.6667, interpMode=RCIM_Cubic, leaveTangentWeight=400.0001, tangentMode=RCTM_User, time=1000, value=1}
         <at 3000 2>
        ]}}}
     CurveIndex=$Engine.Owner.LODDistance
    )
    Uniform Curve Sprite Scale=DataInterface() {propertyValues {Curve {keys [<at 0.5 1 RCIM_Cubic>, {arriveTangent=-5.8778, interpMode=RCIM_Cubic, leaveTangent=-5.8778, tangentMode=RCTM_User, time=1, value=0}]}}}
   )
   &ScaleSpriteSizeBySpeed(
    Max Scale Factor=<0.8 8>
    Sample Scale Factor By Curve=true
    Scale Factor Curve=DataInterface() {propertyValues {Curve {keys [
        {arriveTangent=1.9498, interpMode=RCIM_Cubic, leaveTangent=1.9498, tangentMode=RCTM_User, time=0, value=0}
        {arriveTangent=2.6707, interpMode=RCIM_Cubic, leaveTangent=2.6707, tangentMode=RCTM_User, time=1, value=1}
       ]}}}
    Velocity Threshold=1800
   )
  ]
  rendererClasses [&NiagaraSpriteRendererProperties]
  renderers [{rendererClass=&NiagaraSpriteRendererProperties}]
 }
 ElectricSpirtes {
  emitterUpdateScript [
   &EmitterState()
   &SpawnRate(Spawn Probability=0.6, SpawnRate=12, Use Spawn Probability=true)
   &SpawnBurst_Instantaneous(Spawn Count=2, Use Spawn Probability=false, UseLoopCountLimit=false)
   &SpawnBurst_Instantaneous(Spawn Count=3, Spawn Time=0.3, Use Spawn Probability=false, UseLoopCountLimit=false) {module=SpawnBurst_Instantaneous002}
   &SpawnBurst_Instantaneous(Spawn Count=3, Spawn Group=1, Spawn Time=0.6, Use Spawn Probability=false, UseLoopCountLimit=false) {module=SpawnBurst_Instantaneous001}
  ]
  particleSpawnScript [
   &InitializeParticle(
    Color Maximum=<rgba 7 7 7 1>
    Color Minimum=<rgba 3 3 3 1>
    Color Mode=::NewEnumerator2
    Lifetime Max=0.17
    Lifetime Min=0.12
    Lifetime Mode=::NewEnumerator1
    Mass Mode=::NewEnumerator1
    Sprite Rotation Mode=::NewEnumerator2
    Sprite Size=&UniformRangedVector2D(Minimum=<40 40>, Maximum=<140 140>)
    Sprite Size Mode=::NewEnumerator4
    UsePositionOffset=false
    Write Material Random=true
   ) {module=InitializeParticle001}
   &DynamicMaterialParameters(
    Index 0 Param 1=$Particles.MaterialRandom
    Index 0 Param 2=&Lerp_Float(A=0.7, B=1.25, Alpha=$Particles.MaterialRandom)
    Index 0 Param 4=1
    Module.Param0WriteEnabled=true
    Module.Param1WriteEnabled=true
    Module.Param2WriteEnabled=false
    Module.Param3WriteEnabled=true
   )
  ]
  particleUpdateScript [
   &ParticleState()
   &SpriteRotationRate(Rotation Rate=&RandomRangeFloat(Minimum=-100, Maximum=100))
   &DynamicMaterialParameters(
    Index 0 Param 3=&FloatFromCurve(FloatCurve=DataInterface() {propertyValues {Curve {keys [<at 0.5 0 RCIM_Cubic>, {arriveTangent=3.7979, interpMode=RCIM_Cubic, leaveTangent=3.7979, tangentMode=RCTM_User, time=1, value=1}]}}})
    Module.Param0WriteEnabled=false
    Module.Param0WriteEnabled1=true
    Module.Param1WriteEnabled=false
    Module.Param1WriteEnabled1=true
    Module.Param2WriteEnabled=true
    Module.Param2WriteEnabled1=true
    Module.Param3WriteEnabled=false
    Module.Param3WriteEnabled1=true
    Write Parameter Index 1=true
   ) {module=DynamicMaterialParameters001}
   &ScaleSpriteSize(
    Uniform Curve Scale=&MaskFloatBySpawnGroup(Masked Float=1, Passthrough Float=0.5, Spawn Group Mask=1)
    Uniform Curve Sprite Scale=DataInterface() {propertyValues {Curve {keys [{arriveTangent=0.4536, interpMode=RCIM_Cubic, leaveTangent=0.4536, tangentMode=RCTM_User, time=0, value=0.65}, <at 1 1>]}}}
   )
   &CameraOffset(Camera Offset Amount=/Niagara/DynamicInputs/Math/CalculateParticleRadius(
     Radius Calculation Type:ENiagaraCollisionRadiusOptions=::NewEnumerator0
     Particle Radius:NiagaraFloat=1
     Mesh Dimensions:Vector3f=<1 1 1>
     Method for Calculating Particle Radius:ENiagaraCalculateRadiusOptions=::NewEnumerator1
    ))
   &ScaleColor(
    Scale RGB=&MakeVectorFromLinearColorRGB(Color=&LerpLinearColors(StartColor=<rgba 0.0312 0.6681 1 0.2>, EndColor=<rgba 1.5 0.1028 0.4505 0.2>, LerpFactor=$System.ColorShift))
    ScaleA=false
    ScaleRGB=true
   )
  ]
  rendererClasses [&NiagaraSpriteRendererProperties]
  renderers [{rendererClass=&NiagaraSpriteRendererProperties}]
 }
 Flare_Complex {
  emitterUpdateScript [
   &EmitterState()
   &SpawnBurst_Instantaneous(Use Spawn Probability=false, UseLoopCountLimit=false)
   &SpawnBurst_Instantaneous(Spawn Time=0.3, Use Spawn Probability=false, UseLoopCountLimit=false) {module=SpawnBurst_Instantaneous002}
   &SpawnBurst_Instantaneous(Spawn Group=1, Spawn Time=0.6, Use Spawn Probability=false, UseLoopCountLimit=false) {module=SpawnBurst_Instantaneous001}
  ]
  particleSpawnScript [
   &InitializeParticle(
    Color=&MaskLinearColorBySpawnGroup(Masked Linear Color=<rgba 2 2 2 0.25>, Passthrough Linear Color=<rgba 2 2 2 0.3>, Spawn Group Mask=1)
    Lifetime=&MaskFloatBySpawnGroup(Masked Float=0.4, Passthrough Float=0.2, Spawn Group Mask=1)
    Sprite Size Mode=::NewEnumerator3
    Uniform Sprite Size=&MaskFloatBySpawnGroup(Masked Float=170, Passthrough Float=140, Spawn Group Mask=1)
    UsePositionOffset=false
    Write Material Random=true
   )
   &CameraOffset(Camera Offset Amount=15)
  ]
  particleUpdateScript [
   &ParticleState()
   &ScaleSpriteSize(
    Uniform Curve Scale=&MakeFloatFromVector(Vector=$Engine.Owner.Scale, Channel=::NewEnumerator0)
    Uniform Curve Sprite Scale=DataInterface() {propertyValues {Curve {keys [
        {arriveTangent=23.9365, interpMode=RCIM_Cubic, leaveTangent=23.9365, tangentMode=RCTM_User, time=0, value=0}
        {arriveTangent=2.3333, interpMode=RCIM_Cubic, leaveTangent=2.3333, time=0.1, value=1}
        {arriveTangent=-0.1118, interpMode=RCIM_Cubic, leaveTangent=6.6666, tangentMode=RCTM_Break, time=0.3, value=0.7}
        {arriveTangent=-0.5125, interpMode=RCIM_Cubic, leaveTangent=-0.5125, tangentMode=RCTM_User, time=0.4, value=1}
        {arriveTangent=-1.703, interpMode=RCIM_Cubic, leaveTangent=-1.703, tangentMode=RCTM_User, time=1, value=0.2}
       ]}}}
   )
   &DynamicMaterialParameters(
    Index 0 Param 1=0.6
    Index 0 Param 2=&FloatFromCurve(FloatCurve=DataInterface() {propertyValues {Curve {
        defaultValue=1
        keys [
         {arriveTangent=0.3652, interpMode=RCIM_Cubic, time=0, value=1}
         {arriveTangent=-0.9919, interpMode=RCIM_Cubic, leaveTangent=-0.9919, tangentMode=RCTM_User, time=1, value=0.5}
        ]
       }}})
    Index 3 Param 1=2
    Index 3 Param 2=5
    Module.Index0Float0Write=true
    Module.Index0Float1Write=true
    Module.Index0Float2Write=false
    Module.Index0Float3Write=false
    Module.Index3Float0Write=true
    Module.Index3Float1Write=true
    Module.Index3Float2Write=false
    Module.Index3Float3Write=false
    Write Parameter Index 3=::NewEnumerator1
   )
   &ScaleColor(
    Scale Alpha=&FloatFromCurve(FloatCurve=DataInterface() {propertyValues {Curve {keys [
         {arriveTangent=23.9365, interpMode=RCIM_Cubic, leaveTangent=23.9365, tangentMode=RCTM_User, time=0, value=0}
         {arriveTangent=1.6667, interpMode=RCIM_Cubic, leaveTangent=1.6667, time=0.1, value=1}
         {arriveTangent=-0.1118, interpMode=RCIM_Cubic, leaveTangent=13.5262, tangentMode=RCTM_Break, time=0.3, value=0.5}
         {arriveTangent=-0.5125, interpMode=RCIM_Cubic, leaveTangent=-0.5125, tangentMode=RCTM_User, time=0.4, value=1}
         {arriveTangent=-2.0538, interpMode=RCIM_Cubic, leaveTangent=-2.0538, tangentMode=RCTM_User, time=1, value=0.2}
        ]}}})
    Scale RGB=&MakeVectorFromLinearColorRGB(Color=&LerpLinearColors(StartColor=<rgba 0.0772 0.5401 1 0.2>, EndColor=<rgba 1.5 0.0484 0.6078 0.2>, LerpFactor=$System.ColorShift))
    ScaleA=true
    ScaleRGB=true
   )
   /SetVariables(
    Particles.Darken Color:LinearColor=&MakeLinearColorFromVectorAndFloat(
     "Vector (RGB)"=&MakeVectorFromLinearColorRGB(Color=<rgba 0.0083 0.0099 0.02 1>)
     "Float (Alpha)"=&FloatFromCurve(
      FloatCurve=DataInterface() {propertyValues {Curve {keys [
          {arriveTangent=19.9622, interpMode=RCIM_Cubic, leaveTangent=19.9622, tangentMode=RCTM_User, time=0, value=0}
          {arriveTangentWeight=0.0943, interpMode=RCIM_Cubic, leaveTangentWeight=0.3771, tangentMode=RCTM_User, time=0.1, value=1}
          {arriveTangent=-3.7959, tangentMode=RCTM_User, time=1, value=0}
         ]}}}
      Scale Curve=0.7
     )
    )
    Particles.Darken Size:Vector2f=&Multiply_Vector2D_ByFloat(Float=1.1, Vector2D=$Particles.SpriteSize)
   ) {isSetParameters=true}
   /SetVariables(Particles.Wide Flare Scale:Vector2f=&Vector2DFromCurve(
     Vector2Curve=DataInterface() {propertyValues {
       XCurve {keys [
         {arriveTangent=4.673, interpMode=RCIM_Cubic, leaveTangent=4.673, tangentMode=RCTM_User, time=0, value=0}
         {arriveTangentWeight=0.1536, interpMode=RCIM_Cubic, leaveTangentWeight=0.1152, tangentMode=RCTM_User, time=0.3, value=1}
         {arriveTangent=-3.1914, interpMode=RCIM_Cubic, leaveTangent=-3.1914, tangentMode=RCTM_User, time=0.5, value=0.15}
         {arriveTangent=-2.3891, interpMode=RCIM_Cubic, leaveTangent=-2.3891, tangentMode=RCTM_User, time=0.7, value=0.25}
         {interpMode=RCIM_Cubic, leaveTangent=-1.231, time=1, value=0.2}
        ]}
       YCurve {keys [
         {arriveTangent=10.6903, interpMode=RCIM_Cubic, leaveTangent=10.6903, tangentMode=RCTM_User, time=0, value=0}
         {arriveTangent=0.8, interpMode=RCIM_Cubic, leaveTangent=0.8, time=0.25, value=1}
         {arriveTangent=-1.7663, interpMode=RCIM_Cubic, leaveTangent=-1.7663, tangentMode=RCTM_User, time=0.5, value=0.4}
         {arriveTangent=-4.8806, interpMode=RCIM_Cubic, leaveTangent=-4.8806, tangentMode=RCTM_User, time=0.7, value=1}
         {interpMode=RCIM_Cubic, leaveTangent=-2.2148, time=1, value=0.4}
        ]}
      }}
     ScaleCurve=&Multiply_Vector2D_ByFloat(Float=&MakeFloatFromVector(Vector=$Engine.Owner.Scale, Channel=::NewEnumerator0), Vector2D=<1000 200>)
    )) {isSetParameters=true}
   /SetVariables(Particles.LightColor=&MultiplyLinearColor(Linear Color=$Particles.Color, Scale Factor=<2 0.7 1 50>)) {isSetParameters=true}
  ]
  rendererClasses [&NiagaraSpriteRendererProperties, &NiagaraLightRendererProperties]
  renderers [
   {rendererClass=&NiagaraSpriteRendererProperties}
   {rendererClass=&NiagaraSpriteRendererProperties, rendererIndex=1}
   {rendererClass=&NiagaraSpriteRendererProperties, rendererIndex=2}
   {rendererClass=&NiagaraLightRendererProperties, rendererIndex=3}
  ]
 }
 Fragments {
  emitterUpdateScript [
   &EmitterState()
   &SpawnBurst_Instantaneous(Spawn Count=14, Spawn Time=0.55, Use Spawn Probability=false, UseLoopCountLimit=false)
  ]
  particleSpawnScript [
   &InitializeParticle(
    Color Maximum=<rgba 2 2 2 1>
    Color Minimum=<rgba 1 1 1 1>
    Color Mode=::NewEnumerator2
    Lifetime Max=2.5
    Lifetime Min=1.5
    Lifetime Mode=::NewEnumerator1
    Mesh Renderer Array Visibility Mode=::NewEnumerator3
    Mesh Renderer Info=DataInterface() {propertyValues {MeshRenderer=&self:Fragments_0.NiagaraMeshRendererProperties_0}}
    Mesh Scale Max=<1.6 1.6 1>
    Mesh Scale Min=<0.7 0.7 0.7>
    Mesh Scale Mode=::NewEnumerator2
    Sprite Size Mode=::NewEnumerator1
    Uniform Sprite Size Min=4
    UsePositionOffset=false
    Write Material Random=true
   )
   &ShapeLocation(
    Apply To Particle Position=::NewEnumerator1
    MaskSpawnGroup=false
    Override Seed=false
    Rotation Coordinate Space=::World
    Sphere Radius=&RandomRangeFloat(Minimum=20, Maximum=35)
    UseOwnerScale=false
   )
   /SetVariables(Particles.MeshVis:NiagaraInt32=&RandomRangeInt(Minimum=0, Maximum=1)) {isSetParameters=true}
  ]
  particleUpdateScript [
   &ParticleState()
   &ScaleColor(
    Scale Alpha=&FloatFromCurve(FloatCurve=DataInterface() {propertyValues {Curve {
        defaultValue=0
        keys [
         {arriveTangent=17.292, interpMode=RCIM_Cubic, leaveTangent=17.292, tangentMode=RCTM_User, time=0, value=0}
         {arriveTangentWeight=0.1389, interpMode=RCIM_Cubic, leaveTangentWeight=0.2778, tangentMode=RCTM_User, time=0.1, value=1}
         {arriveTangentWeight=0.1667, interpMode=RCIM_Cubic, leaveTangentWeight=0.0833, tangentMode=RCTM_User, time=0.5, value=1}
         <at 1 0 RCIM_Cubic>
        ]
       }}})
    Scale RGB=&MakeVectorFromLinearColorRGB(Color=&LerpLinearColors(StartColor=<rgba 0.0772 0.5401 1 0.2>, EndColor=<rgba 1.5 0.0484 0.6078 0.2>, LerpFactor=$System.ColorShift))
    ScaleA=true
    ScaleRGB=true
   )
   &UpdateMeshOrientation(
    Rotation Rate=&RandomRangeFloat(Minimum=0.4, Maximum=0.7)
    Rotation Vector=&RandomVector(Vector Scale=1)
   )
   &ScaleSpriteSize(Uniform Curve Sprite Scale=DataInterface() {propertyValues {Curve {
       defaultValue=0
       keys [{interpMode=RCIM_Cubic, tangentMode=RCTM_User, time=0.5, value=1}, {arriveTangent=-3.9854, tangentMode=RCTM_User, time=1, value=0}]
      }}})
   &InterpolateOverTime(
    Rate Of Change=12
    Target Position Value:NiagaraPosition=&AddVectorToPosition(
     Position=$Engine.Owner.Position
     Vector=&RotateVectorByQuaternion(
      Vector To Rotate=&RandomVector(Vector Scale=&FloatFromCurve(
        FloatCurve=DataInterface() {propertyValues {Curve {
           defaultValue=1
           keys [
            {arriveTangent=2.4672, interpMode=RCIM_Cubic, leaveTangent=2.4672, tangentMode=RCTM_User, time=0, value=0}
            {arriveTangent=0.0022, interpMode=RCIM_Cubic, leaveTangent=0.0022, tangentMode=RCTM_User, time=0.7, value=1}
            <at 3 0.25>
           ]
          }}}
        CurveIndex=$Particles.Age
        Scale Curve=&RandomRangeFloat(Minimum=35, Maximum=60)
       ))
      Quaternion=&MakeQuaternion(
       Quaternion From=0
       Axis=&RandomVector(Vector Scale=1)
       Angle=&FloatFromCurve(
        FloatCurve=DataInterface() {propertyValues {Curve {
           defaultValue=1
           keys [
            {arriveTangent=0.9195, interpMode=RCIM_Cubic, leaveTangent=0.9195, tangentMode=RCTM_User, time=0, value=0}
            {tangentMode=RCTM_User, time=3, value=1}
           ]
          }}}
        CurveIndex=$Particles.Age
        Scale Curve=&RandomRangeFloat(Minimum=0.1, Maximum=0.3)
       )
       From=?
       To=?
       Angle Type=::NewEnumerator1
       Normalize Vectors=false
      )
     )
    )
    Variable Type=::NewEnumerator6
   )
   /SetVariables(Particles.Position=$Particles.InterpolateOverTime.InterpolateOverTime_Position.Moving Average) {isSetParameters=true}
  ]
  rendererClasses [&NiagaraSpriteRendererProperties, &NiagaraMeshRendererProperties]
  renderers [
   {rendererClass=&NiagaraSpriteRendererProperties}
   {rendererClass=&NiagaraMeshRendererProperties, rendererIndex=1}
  ]
 }
}
summary {
 systemName=NS_LimeRock_RiftDevice_Burst
 emitters [
  {emitterName=Flare_Complex, rendererClasses [&NiagaraSpriteRendererProperties, &NiagaraLightRendererProperties]}
  {emitterName=Fragments, rendererClasses [&NiagaraSpriteRendererProperties, &NiagaraMeshRendererProperties]}
  {emitterName=BaseSparks, rendererClasses [&NiagaraSpriteRendererProperties]}
  {emitterName=ElectricSpirtes, rendererClasses [&NiagaraSpriteRendererProperties]}
 ]
 userVariables []
}
