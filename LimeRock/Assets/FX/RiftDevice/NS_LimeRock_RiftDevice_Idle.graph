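# Niagara system for the LimeRock Rift Device's idle effect: a glowing flare with drifting fragments around it.
# Flare_Complex burst-spawns a flare that is not killed at lifetime end, layering four sprite renderers and a light with curve-driven glow, darkening and wide-streak parameters.
# Fragments continuously spawns sprite and mesh shards that chase randomized rotated targets around the owner through smoothed interpolation.
# User.ColorShift blends blue-violet toward pink-red, while User.IntensityLerp raises size, flare brightness, light radius, spawn rate and motion.
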NiagaraSystem /LimeRock/Assets/FX/RiftDevice/NS_LimeRock_RiftDevice_Idle
dependencies {
 usedDataInterfaces [
  &NiagaraDataInterfaceCurve
  &NiagaraDataInterfaceVector2DCurve
  &NiagaraDataInterfaceMeshRendererInfo
 ]
 usedDynamicInputs []
 usedModules [&SystemState]
}
emitters {
 Flare_Complex {
  emitterUpdateScript [
   &EmitterState(
    Inactive Response=::NewEnumerator1
    Life Cycle Mode=::NewEnumerator1
    Loop Behavior=::NewEnumerator1
    Loop Duration Mode=::NewEnumerator1
    UseLoopDelay=false
   )
   &SpawnBurst_Instantaneous(Use Spawn Probability=false, UseLoopCountLimit=false)
  ]
  particleSpawnScript [
   &InitializeParticle(
    Color=<rgba 2 2 2 0.2>
    Lifetime=0.5
    Sprite Size Mode=::NewEnumerator3
    Uniform Sprite Size=90
    UsePositionOffset=false
    Write Material Random=true
   )
   &CameraOffset(Camera Offset Amount=15)
  ]
  particleUpdateScript [
   &ParticleState(Kill Particles When Lifetime Has Elapsed=false)
   &ScaleSpriteSize(
    Uniform Curve Scale=&Multiply_Float(
     A=&MakeFloatFromVector(Vector=$Engine.Owner.Scale, Channel=::NewEnumerator0)
     B=&Lerp_Float(A=0.7, B=2, Alpha=$User.IntensityLerp)
    )
    Uniform Curve Sprite Scale=DataInterface() {propertyValues {Curve {keys [
        {arriveTangent=5.8052, interpMode=RCIM_Cubic, leaveTangent=5.8052, tangentMode=RCTM_User, time=0, value=0}
        {arriveTangent=-0.9873, interpMode=RCIM_Cubic, leaveTangent=-0.9873, tangentMode=RCTM_User, time=0.4, value=1.25}
        {interpMode=RCIM_Cubic, leaveTangent=0.7472, time=1, value=1}
       ]}}}
   )
   &DynamicMaterialParameters(
    Index 0 Param 1=0.6
    Index 0 Param 2=&FloatFromCurve(FloatCurve=DataInterface() {propertyValues {Curve {defaultValue=1, keys [<at 0 0 RCIM_Cubic>, {tangentMode=RCTM_User, time=1, value=0.7}]}}})
    Index 3 Param 1=2
    Index 3 Param 2=5
    Module.Index0Param0Write=true
    Module.Index0Param1Write=true
    Module.Index0Param2Write=true
    Module.Index0Param3Write=false
    Module.Index3Param0Write=true
    Module.Index3Param1Write=true
    Module.Index3Param2Write=false
    Module.Index3Param3Write=false
    Write Parameter Index 3=true
   )
   &ScaleColor(
    Scale Alpha=&FloatFromCurve(FloatCurve=DataInterface() {propertyValues {Curve {keys [
         <at 0 0 RCIM_Cubic>
         {arriveTangentWeight=0.0943, interpMode=RCIM_Cubic, leaveTangentWeight=0.3771, tangentMode=RCTM_User, time=0.2, value=1}
         <at 1 1>
        ]}}})
    Scale RGB=&Multiply_VectorByFloat(
     Vector=&MakeVectorFromLinearColorRGB(Color=&LerpLinearColors(StartColor=<rgba 0.0772 0.5401 1 0.2>, EndColor=<rgba 0.5 0.0161 0.2026 0.2>, LerpFactor=$User.ColorShift))
     Float=&Lerp_Float(A=1, B=3, Alpha=$User.IntensityLerp)
    )
    ScaleA=true
    ScaleRGB=true
   )
   /SetVariables(
    Particles.Darken Color:LinearColor=&MakeLinearColorFromVectorAndFloat(
     "Vector (RGB)"=&MakeVectorFromLinearColorRGB(Color=<rgba 0.0083 0.0099 0.02 1>)
     "Float (Alpha)"=&FloatFromCurve(FloatCurve=DataInterface() {propertyValues {Curve {
         defaultValue=1
         keys [{interpMode=RCIM_Cubic, tangentMode=RCTM_User, time=0, value=0}, {interpMode=RCIM_Cubic, tangentMode=RCTM_User, time=1, value=1}]
        }}})
    )
    Particles.Darken Size:Vector2f=&Multiply_Vector2D_ByFloat(Float=1.2, Vector2D=$Particles.SpriteSize)
   ) {isSetParameters=true}
   /SetVariables(
    Particles.Glow Camera Offset=100
    Particles.Glow Color=&MultiplyLinearColor(
     Linear Color=&LerpLinearColors(StartColor=<rgba 0.3154 0.1557 1 0.2>, EndColor=<rgba 0.5 0.0149 0.0244 0.2>, LerpFactor=$User.ColorShift)
     Scale Factor=&MakeFloat4FromVectorAndFloat(
      "Vector (RGB)"=&VectorFromFloat(Value=&Lerp_Float(A=1, B=3, Alpha=$User.IntensityLerp))
      "Float (Alpha)"=&FloatFromCurve(
       FloatCurve=DataInterface() {propertyValues {Curve {keys [
           {arriveTangent=3.5878, interpMode=RCIM_Cubic, leaveTangent=3.5878, tangentMode=RCTM_User, time=0, value=0.35}
           {arriveTangent=0.3, interpMode=RCIM_Cubic, leaveTangent=0.3, time=0.3, value=1}
           {arriveTangent=3.4933, interpMode=RCIM_Cubic, leaveTangent=3.4933, tangentMode=RCTM_User, time=0.5, value=0.5}
           {interpMode=RCIM_Cubic, leaveTangent=-0.0405, time=1, value=1}
          ]}}}
       Scale Curve=0.2
      )
     )
    )
    Particles.Glow Size=&Multiply_Vector2D_ByFloat(Float=1.2, Vector2D=$Particles.SpriteSize)
   ) {isSetParameters=true}
   /SetVariables(Particles.Wide Flare Scale:Vector2f=&Vector2DFromCurve(
     Vector2Curve=DataInterface() {propertyValues {
       XCurve {keys [
         {arriveTangent=4.673, interpMode=RCIM_Cubic, leaveTangent=4.673, tangentMode=RCTM_User, time=0, value=0}
         {arriveTangentWeight=0.1536, interpMode=RCIM_Cubic, leaveTangentWeight=0.1152, tangentMode=RCTM_User, time=0.3, value=1}
         {arriveTangent=-3.1914, interpMode=RCIM_Cubic, leaveTangent=-3.1914, tangentMode=RCTM_User, time=0.5, value=0.15}
         {arriveTangent=-2.3891, interpMode=RCIM_Cubic, leaveTangent=-2.3891, tangentMode=RCTM_User, time=0.7, value=0.25}
         {interpMode=RCIM_Cubic, leaveTangent=-1.231, time=1, value=0.2}
        ]}
       YCurve {keys [
         {arriveTangent=10.6903, interpMode=RCIM_Cubic, leaveTangent=10.6903, tangentMode=RCTM_User, time=0, value=0}
         {arriveTangent=0.8, interpMode=RCIM_Cubic, leaveTangent=0.8, time=0.25, value=1}
         {arriveTangent=-1.7663, interpMode=RCIM_Cubic, leaveTangent=-1.7663, tangentMode=RCTM_User, time=0.5, value=0.4}
         {arriveTangent=-4.8806, interpMode=RCIM_Cubic, leaveTangent=-4.8806, tangentMode=RCTM_User, time=0.7, value=1}
         {interpMode=RCIM_Cubic, leaveTangent=-2.2148, time=1, value=0.4}
        ]}
      }}
     ScaleCurve=&Multiply_Vector2D_ByFloat(
      Float=&Multiply_Float(
       A=&MakeFloatFromVector(Vector=$Engine.Owner.Scale, Channel=::NewEnumerator0)
       B=&Lerp_Float(A=0.7, B=2, Alpha=$User.IntensityLerp)
      )
      Vector2D=<2500 300>
     )
    )) {isSetParameters=true}
   /SetVariables(
    Particles.LightColor=&MultiplyLinearColor(Linear Color=$Particles.Color, Scale Factor=<2 0.7 1 100>)
    Particles.LightRadius=&Lerp_Float(A=150, B=300, Alpha=$User.IntensityLerp)
   ) {isSetParameters=true}
  ]
  rendererClasses [&NiagaraSpriteRendererProperties, &NiagaraLightRendererProperties]
  renderers [
   {rendererClass=&NiagaraSpriteRendererProperties}
   {rendererClass=&NiagaraSpriteRendererProperties, rendererIndex=1}
   {rendererClass=&NiagaraSpriteRendererProperties, rendererIndex=2}
   {rendererClass=&NiagaraSpriteRendererProperties, rendererIndex=3}
   {rendererClass=&NiagaraLightRendererProperties, rendererIndex=4}
  ]
 }
 Fragments {
  emitterUpdateScript [
   &EmitterState()
   &SpawnRate(SpawnRate=&Lerp_Float(A=2, B=4, Alpha=$User.IntensityLerp), Use Spawn Probability=false)
  ]
  particleSpawnScript [
   &InitializeParticle(
    Color Maximum=<rgba 2 2 2 1>
    Color Minimum=<rgba 1 1 1 1>
    Color Mode=::NewEnumerator2
    Lifetime Max=4.5
    Lifetime Min=3
    Lifetime Mode=::NewEnumerator1
    Mesh Renderer Array Visibility Mode=::NewEnumerator3
    Mesh Renderer Info=DataInterface() {propertyValues {MeshRenderer=&self:Fragments_0.NiagaraMeshRendererProperties_0}}
    Mesh Scale Max=<0.7 0.7 1>
    Mesh Scale Min=<0.4 0.4 1>
    Mesh Scale Mode=::NewEnumerator2
    Sprite Size Mode=::NewEnumerator1
    Uniform Sprite Size Min=4
    UsePositionOffset=false
    Write Material Random=true
   )
   /SetVariables(Particles.OrbitSpeed:NiagaraFloat=&Multiply_Float(
     A=&RandomRangeFloat(Minimum=0.08, Maximum=0.2)
     B=&Lerp_Float(A=1, B=1.7, Alpha=$User.IntensityLerp)
    )) {isSetParameters=true}
   &ShapeLocation(
    Apply To Particle Position=::NewEnumerator1
    MaskSpawnGroup=false
    Override Seed=false
    Rotation Coordinate Space=::World
    Sphere Radius=&RandomRangeFloat(Minimum=20, Maximum=35)
    UseOwnerScale=false
   )
   /SetVariables(Particles.MeshVis:NiagaraInt32=&RandomRangeInt(Minimum=0, Maximum=1)) {isSetParameters=true}
  ]
  particleUpdateScript [
   &ParticleState()
   &ScaleMeshSize(Scale Factor=&VectorFromFloat(Value=&Lerp_Float(A=1, B=1.4, Alpha=$User.IntensityLerp)))
   &ScaleColor(
    Scale Alpha=&FloatFromCurve(FloatCurve=DataInterface() {propertyValues {Curve {
        defaultValue=0
        keys [
         {arriveTangent=4.5168, interpMode=RCIM_Cubic, leaveTangent=4.5168, tangentMode=RCTM_User, time=0, value=0}
         {arriveTangentWeight=0.1389, interpMode=RCIM_Cubic, leaveTangentWeight=0.2778, tangentMode=RCTM_User, time=0.25, value=1}
         {arriveTangentWeight=0.1667, interpMode=RCIM_Cubic, leaveTangentWeight=0.0833, tangentMode=RCTM_User, time=0.75, value=1}
         {arriveTangent=-5.3936, tangentMode=RCTM_User, time=1, value=0}
        ]
       }}})
    Scale RGB=&MakeVectorFromLinearColorRGB(Color=&LerpLinearColors(StartColor=<rgba 0.0772 0.5401 1 0.2>, EndColor=<rgba 1.5 0.0484 0.6078 0.2>, LerpFactor=$User.ColorShift))
    ScaleA=true
    ScaleRGB=true
   )
   &UpdateMeshOrientation(
    Rotation Rate=&Multiply_Float(
     A=&RandomRangeFloat(Minimum=0.2, Maximum=0.5)
     B=&Lerp_Float(A=1, B=4, Alpha=$User.IntensityLerp)
    )
    Rotation Vector=&RandomVector(Vector Scale=1)
   )
   &ScaleSpriteSize(
    Uniform Curve Scale=&Lerp_Float(A=1, B=2, Alpha=$User.IntensityLerp)
    Uniform Curve Sprite Scale=DataInterface() {propertyValues {Curve {
       defaultValue=0
       keys [
        {arriveTangent=2.5266, interpMode=RCIM_Cubic, leaveTangent=2.5266, tangentMode=RCTM_User, time=0, value=0}
        {interpMode=RCIM_Cubic, tangentMode=RCTM_User, time=0.5, value=1}
        {arriveTangent=-3.9854, tangentMode=RCTM_User, time=1, value=0}
       ]
      }}}
   )
   &InterpolateOverTime(
    Rate Of Change=7
    Target Position Value:NiagaraPosition=&AddVectorToPosition(
     Position=$Engine.Owner.Position
     Vector=&RotateVectorByQuaternion(
      Vector To Rotate=&RandomVector(Vector Scale=&Multiply_Float(
        A=&RandomRangeFloat(Minimum=15, Maximum=25)
        B=&Lerp_Float(A=1, B=2, Alpha=$User.IntensityLerp)
       ))
      Quaternion=&MakeQuaternion(
       Quaternion From=0
       Axis=&RandomVector(Vector Scale=1)
       Angle=&Multiply_Float(A=$Particles.Age, B=&Add_Float(A=$Particles.OrbitSpeed, B=$User.IntensityLerp))
       From=?
       To=?
       Angle Type=::NewEnumerator1
       Normalize Vectors=false
      )
     )
    )
    Variable Type=::NewEnumerator6
   )
   /SetVariables(Particles.Position=$Particles.InterpolateOverTime.InterpolateOverTime_Position.Moving Average) {isSetParameters=true}
  ]
  rendererClasses [&NiagaraSpriteRendererProperties, &NiagaraMeshRendererProperties]
  renderers [
   {rendererClass=&NiagaraSpriteRendererProperties}
   {rendererClass=&NiagaraMeshRendererProperties, rendererIndex=1}
  ]
 }
}
summary {
 systemName=NS_LimeRock_RiftDevice_Idle
 emitters [
  {emitterName=Flare_Complex, rendererClasses [&NiagaraSpriteRendererProperties, &NiagaraLightRendererProperties]}
  {emitterName=Fragments, rendererClasses [&NiagaraSpriteRendererProperties, &NiagaraMeshRendererProperties]}
 ]
 userVariables [
  User.ColorShift {type=NiagaraFloat, defaultValue=0, description=""}
  User.IntensityLerp {type=NiagaraFloat, defaultValue=0, description=""}
 ]
}
