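# Niagara buildup effect for the LimeRock rift device's "rift in", growing over about 2.5 seconds.
# LargeCracks and SmallCrack are growing meshes animated through dynamic material parameters.
# FlareLight drives four sprite renderers, computing glow, darken and wide-flare attributes from curves.
# ElectricArcs fires staggered, short-lived, growing sprite bursts; User.Pink switches their base color, and User.SizeNormalized scales arc size and flare camera offset.
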NiagaraSystem /LimeRock/Assets/FX/RiftDevice/NS_LimeRock_RiftIn_Buildup
dependencies {
 usedDataInterfaces [&NiagaraDataInterfaceCurve, &NiagaraDataInterfaceVector2DCurve, &NiagaraDataInterfaceColorCurve]
 usedDynamicInputs [&SimulationPosition]
 usedModules [&SystemState]
}
emitters {
 ElectricArcs {
  emitterUpdateScript [
   &EmitterState()
   &SpawnBurst_Instantaneous(Spawn Group=1, Spawn Time=0.25, Use Spawn Probability=false, UseLoopCountLimit=false)
   &SpawnBurst_Instantaneous(Spawn Count=2, Spawn Time=1, Use Spawn Probability=false, UseLoopCountLimit=false) {module=SpawnBurst_Instantaneous001}
   &SpawnBurst_Instantaneous(Spawn Count=2, Spawn Time=1.7, Use Spawn Probability=false, UseLoopCountLimit=false) {module=SpawnBurst_Instantaneous002}
   &SpawnBurst_Instantaneous(Spawn Count=3, Spawn Time=2, Use Spawn Probability=false, UseLoopCountLimit=false) {module=SpawnBurst_Instantaneous003}
   &SpawnBurst_Instantaneous(Spawn Count=3, Spawn Time=2.3, Use Spawn Probability=false, UseLoopCountLimit=false) {module=SpawnBurst_Instantaneous004}
  ]
  particleSpawnScript [
   &InitializeParticle(
    Color=&MakeCustomLinearColorFromBool(Bool=$User.Pink, True Color=<rgba 2 0.0849 1.9761 1>, False Color=<rgba 0.0813 0.6736 2 1>)
    Lifetime=&UniformRangedFloat(Minimum=0.09, Maximum=0.125)
    Mass Mode=::NewEnumerator1
    Sprite Rotation Mode=::NewEnumerator2
    Sprite Size=&Multiply_Vector2D_ByFloat(
     Float=&FloatFromCurve(
      FloatCurve=DataInterface() {propertyValues {Curve {defaultValue=1, keys [<at 0 0.4 RCIM_Cubic>, {tangentMode=RCTM_User, time=1, value=1}]}}}
      CurveIndex=$Emitter.NormalizedLoopAge
     )
     Vector2D=&UniformRangedVector2D(Minimum=<200 200>, Maximum=<400 400>)
    )
    Sprite Size Mode=::NewEnumerator4
    UsePositionOffset=false
    Write Material Random=true
   ) {module=InitializeParticle001}
   &DynamicMaterialParameters(
    Index 0 Param 1=$Particles.MaterialRandom
    Index 0 Param 2=&Lerp_Float(A=0.75, B=1.2, Alpha=$Particles.MaterialRandom)
    Index 0 Param 4=1
    Module.Param0WriteEnabled=true
    Module.Param1WriteEnabled=true
    Module.Param2WriteEnabled=false
    Module.Param3WriteEnabled=true
   )
   &CameraOffset(Camera Offset Amount=20)
   &ShapeLocation(MaskSpawnGroup=false, Override Seed=false, Sphere Surface Distribution=1, UseOwnerScale=true)
  ]
  particleUpdateScript [
   &ParticleState()
   &DynamicMaterialParameters(
    Index 0 Param 3=&FloatFromCurve(FloatCurve=DataInterface() {propertyValues {Curve {keys [<at 0.5 0 RCIM_Cubic>, {interpMode=RCIM_Cubic, leaveTangent=3.7979, time=1, value=1}]}}})
    Index 1 Param 2=&FloatFromCurve(FloatCurve=DataInterface() {propertyValues {Curve {keys [<at 0.25 1 RCIM_Cubic>,<at 0.5 0>]}}}, CurveIndex=$System.Age, Scale Curve=0.5)
    Index 1 Param 3=&FloatFromCurve(
     FloatCurve=DataInterface() {propertyValues {Curve {keys [
         {arriveTangent=4.106, interpMode=RCIM_Cubic, leaveTangent=4.106, tangentMode=RCTM_User, time=0.05, value=0}
         {tangentMode=RCTM_User, time=0.35, value=1}
        ]}}}
     CurveIndex=$System.Age
    )
    Index 1 Param 4=100
    Module.Param0WriteEnabled=false
    Module.Param0WriteEnabled1=true
    Module.Param1WriteEnabled=false
    Module.Param1WriteEnabled1=true
    Module.Param2WriteEnabled=true
    Module.Param2WriteEnabled1=true
    Module.Param3WriteEnabled=false
    Module.Param3WriteEnabled1=true
    Write Parameter Index 1=true
   ) {module=DynamicMaterialParameters001}
   &ScaleSpriteSize(
    Uniform Curve Scale=&Lerp_Float(A=0.5, B=1, Alpha=$User.SizeNormalized)
    Uniform Curve Sprite Scale=DataInterface() {propertyValues {Curve {keys [
        {arriveTangent=0.5666, interpMode=RCIM_Cubic, leaveTangent=0.5666, tangentMode=RCTM_User, time=0, value=0.75}
        {tangentMode=RCTM_User, time=1, value=1}
       ]}}}
   )
   &ScaleColor(
    Scale RGB=&MakeVectorFromLinearColorRGB(Color=&ColorFromCurve(
      ColorCurve=DataInterface() {propertyValues {
        AlphaCurve {keys []}
        BlueCurve {keys [<at 0.0806 4 RCIM_Cubic>,<at 0.4309 1.5>]}
        GreenCurve {keys [<at 0.0806 3.5603 RCIM_Cubic>,<at 0.4309 0.2162>]}
        RedCurve {keys [<at 0.0806 1.2532 RCIM_Cubic>,<at 0.4309 0.1094>]}
       }}
      CurveIndex=$Emitter.Age
     ))
    ScaleA=false
    ScaleRGB=true
   ) {module=ScaleColor001}
  ]
  rendererClasses [&NiagaraSpriteRendererProperties]
  renderers [{rendererClass=&NiagaraSpriteRendererProperties}]
 }
 FlareLight {
  emitterUpdateScript [
   &EmitterState()
   &SpawnBurst_Instantaneous(Spawn Time=0.1, Use Spawn Probability=false, UseLoopCountLimit=false)
  ]
  particleSpawnScript [
   &InitializeParticle(
    Color=<rgba 0.3252 2.6943 8 1>
    Lifetime=2.5
    Sprite Size Mode=::NewEnumerator3
    Uniform Sprite Size=200
    UsePositionOffset=false
    Write Material Random=true
   )
   /SetVariables(Particles.GlowCamOffset=500) {isSetParameters=true}
  ]
  particleUpdateScript [
   &ParticleState()
   &ScaleSpriteSize(Uniform Curve Sprite Scale=DataInterface() {propertyValues {Curve {keys [
        {arriveTangent=5.2794, interpMode=RCIM_Cubic, time=0, value=1}
        {arriveTangent=0.7472, interpMode=RCIM_Cubic, leaveTangent=0.7472, tangentMode=RCTM_User, time=0.9, value=4}
        {arriveTangent=-99.6127, interpMode=RCIM_Cubic, leaveTangent=-99.6127, tangentMode=RCTM_User, time=0.9961, value=0}
       ]}}})
   &CameraOffset(Camera Offset Amount=&Multiply_Float(
     A=&Lerp_Float(A=15, B=100, Alpha=$Particles.NormalizedAge)
     B=&Lerp_Float(A=0.4, B=1, Alpha=$User.SizeNormalized)
    ))
   /SetVariables(Particles.DarkenColor:LinearColor=&MakeLinearColorFromVectorAndFloat(
     "Vector (RGB)"=&MakeVectorFromLinearColorRGB(Color=<rgba 0.01 0.0143 0.02 1>)
     "Float (Alpha)"=&FloatFromCurve(FloatCurve=DataInterface() {propertyValues {Curve {keys [
          {interpMode=RCIM_Cubic, leaveTangentWeight=0.3333, tangentMode=RCTM_User, time=0, value=0}
          {arriveTangentWeight=0.3333, interpMode=RCIM_Cubic, tangentMode=RCTM_User, time=1, value=1}
         ]}}})
    )) {isSetParameters=true}
   /SetVariables(Particles.DarkenSize:Vector2f=&Multiply_Vector2D_ByFloat(Float=1.4, Vector2D=$Particles.SpriteSize)) {isSetParameters=true}
   /SetVariables(Particles.WideFlareSize:Vector2f=&Vector2DFromCurve(
     Vector2Curve=DataInterface() {propertyValues {
       XCurve {keys [{arriveTangent=4.673, interpMode=RCIM_Cubic, time=0, value=0}, {interpMode=RCIM_Cubic, leaveTangent=-1.231, time=1, value=1}]}
       YCurve {keys [{arriveTangent=10.6903, interpMode=RCIM_Cubic, time=0, value=0}, {interpMode=RCIM_Cubic, leaveTangent=-2.2148, time=1, value=1}]}
      }}
     ScaleCurve=<1600 150>
    )) {isSetParameters=true}
   /SetVariables(Particles.GlowSize=&Vector2DFromFloat(Value=&FloatFromCurve(
      FloatCurve=DataInterface() {propertyValues {Curve {
         defaultValue=1
         keys [
          {arriveTangent=1.2592, interpMode=RCIM_Cubic, leaveTangent=1.2592, tangentMode=RCTM_User, time=0, value=0.25}
          {arriveTangent=0.9471, tangentMode=RCTM_User, time=1, value=1}
         ]
        }}}
      Scale Curve=1000
     ))) {isSetParameters=true}
   &DynamicMaterialParameters(
    Index 0 Param 1=1
    Index 0 Param 2=&Lerp_Float(
     A=1.2
     B=0
     Alpha=&FloatFromCurve(FloatCurve=DataInterface() {propertyValues {Curve {keys [
          {arriveTangent=-1.3124, interpMode=RCIM_Cubic, leaveTangentWeight=0.55, time=0, value=1}
          {arriveTangent=-1.8194, arriveTangentWeight=0.3578, interpMode=RCIM_Cubic, leaveTangent=-1.8194, tangentMode=RCTM_User, time=1, value=0}
         ]}}})
    )
    Index 3 Param 1=2
    Index 3 Param 2=5
    Module.Index0Float0Write=true
    Module.Index0Float1Write=true
    Module.Index0Float2Write=true
    Module.Index0Float3Write=false
    Module.Index3Float0Write=true
    Module.Index3Float1Write=true
    Module.Index3Float2Write=false
    Module.Index3Float3Write=false
    Write Parameter Index 3=::NewEnumerator1
   )
   &ScaleColor(
    Scale Alpha=&FloatFromCurve(FloatCurve=DataInterface() {propertyValues {Curve {keys [
         {arriveTangent=1.4983, interpMode=RCIM_Cubic, leaveTangent=1.4983, leaveTangentWeight=0.3333, tangentMode=RCTM_User, time=0, value=0}
         {arriveTangent=2.8903, tangentMode=RCTM_User, time=1, value=1}
        ]}}})
    ScaleA=true
    ScaleRGB=false
   )
   /SetVariables(Particles.GlowColor=&MultiplyLinearColor(
     Linear Color=$Particles.Color
     Scale Factor=&MakeFloat4FromVectorAndFloat(
      "Vector (RGB)"=<1 1 1>
      "Float (Alpha)"=&FloatFromCurve(
       FloatCurve=DataInterface() {propertyValues {Curve {keys [
           {arriveTangent=3.5878, interpMode=RCIM_Cubic, time=0, value=0}
           {arriveTangentWeight=0.0807, interpMode=RCIM_Cubic, leaveTangentWeight=0.3549, time=0.97, value=1}
           {arriveTangent=-74.6008, interpMode=RCIM_Cubic, leaveTangent=-74.6008, tangentMode=RCTM_User, time=1, value=0}
          ]}}}
       Scale Curve=0.05
      )
     )
    )) {isSetParameters=true}
  ]
  rendererClasses [&NiagaraSpriteRendererProperties]
  renderers [
   {rendererClass=&NiagaraSpriteRendererProperties}
   {rendererClass=&NiagaraSpriteRendererProperties, rendererIndex=1}
   {rendererClass=&NiagaraSpriteRendererProperties, rendererIndex=2}
   {rendererClass=&NiagaraSpriteRendererProperties, rendererIndex=3}
  ]
 }
 LargeCracks {
  emitterUpdateScript [
   &EmitterState()
   &SpawnBurst_Instantaneous(Use Spawn Probability=false, UseLoopCountLimit=false)
  ]
  particleSpawnScript [
   &InitializeParticle(
    Color=<rgba 2.1324 2.7967 4 1>
    Lifetime=2.5
    Mass Mode=::NewEnumerator1
    Mesh Scale=&Multiply_VectorByFloat(Vector=&RandomRangeVector(Minimum=<5 5 5>, Maximum=<9 9 9>), Float=1.5)
    Mesh Scale Mode=::NewEnumerator4
    Sprite Size Mode=::NewEnumerator4
    UsePositionOffset=false
    Write Material Random=true
   ) {module=InitializeParticle001}
   &InitialMeshOrientation(Use Rotation Vector=false)
   &CameraOffset(Camera Offset Amount=30)
  ]
  particleUpdateScript [
   &ParticleState()
   &ScaleColor(
    Scale RGB=&VectorFromFloat(Value=&FloatFromCurve(FloatCurve=DataInterface() {propertyValues {Curve {keys [
          {arriveTangentWeight=0.05, interpMode=RCIM_Cubic, leaveTangentWeight=0.1667, tangentMode=RCTM_User, time=0.5, value=1}
          {arriveTangent=3.378, arriveTangentWeight=0.1667, interpMode=RCIM_Cubic, leaveTangent=3.378, tangentMode=RCTM_User, time=1, value=3}
         ]}}}))
    ScaleA=false
    ScaleRGB=true
   )
   &ScaleMeshSize(Scale Factor=&VectorFromFloat(Value=&FloatFromCurve(FloatCurve=DataInterface() {propertyValues {Curve {keys [
          {arriveTangent=1.9392, interpMode=RCIM_Cubic, time=0, value=0.5}
          {arriveTangent=0.9167, interpMode=RCIM_Cubic, leaveTangent=0.9167, tangentMode=RCTM_User, time=0.5, value=1}
         ]}}})))
   &DynamicMaterialParameters(
    Index 0 Param 1=&FloatFromCurve(
     FloatCurve=DataInterface() {propertyValues {Curve {
        defaultValue=1
        keys [
         {interpMode=RCIM_Cubic, leaveTangentWeight=0.3333, tangentMode=RCTM_User, time=0, value=0}
         {arriveTangent=1.5072, tangentMode=RCTM_User, time=1, value=1}
        ]
       }}}
     Scale Curve=0.5
    )
    Index 0 Param 2=&FloatFromCurve(
     FloatCurve=DataInterface() {propertyValues {Curve {keys [
         {arriveTangent=6.2648, interpMode=RCIM_Cubic, leaveTangent=6.2648, tangentMode=RCTM_User, time=0, value=0}
         {arriveTangentWeight=0.0536, interpMode=RCIM_Cubic, leaveTangentWeight=0.1786, tangentMode=RCTM_User, time=0.25, value=1}
         {arriveTangentWeight=0.6601, interpMode=RCIM_Cubic, leaveTangentWeight=0.088, tangentMode=RCTM_User, time=0.97, value=1}
         {arriveTangent=-14.6836, interpMode=RCIM_Cubic, leaveTangent=-14.6836, tangentMode=RCTM_User, time=1, value=0}
        ]}}}
     Scale Curve=0.8
    )
    Index 0 Param 3=&Add_Float(
     A=&FloatFromCurve(FloatCurve=DataInterface() {propertyValues {Curve {keys [{arriveTangent=0.6161, interpMode=RCIM_Cubic, time=0, value=-0.1}, {arriveTangent=0.2234, tangentMode=RCTM_User, time=0.5, value=0}]}}})
     B=&RandomRangeFloat(Minimum=0, Maximum=1)
    )
    Module.Index0Float0Write=true
    Module.Index0Float1Write=true
    Module.Index0Float2Write=true
    Module.Index0Float3Write=false
   )
  ]
  rendererClasses [&NiagaraMeshRendererProperties]
  renderers [{rendererClass=&NiagaraMeshRendererProperties, rendererIndex=0}]
 }
 SmallCrack {
  emitterUpdateScript [
   &EmitterState()
   &SpawnBurst_Instantaneous(Use Spawn Probability=false, UseLoopCountLimit=false)
  ]
  particleSpawnScript [
   &InitializeParticle(
    Color=<rgba 2.0772 4.8329 10 1.5>
    Lifetime=2.5
    Mesh Scale Mode=::NewEnumerator3
    Mesh Uniform Scale=10
    UsePositionOffset=false
    Write Material Random=true
   )
   &InitialMeshOrientation(Use Rotation Vector=false)
   &CameraOffset(Camera Offset Amount=15)
  ]
  particleUpdateScript [
   &ParticleState()
   &ScaleMeshSize(Scale Factor=&VectorFromFloat(Value=&FloatFromCurve(FloatCurve=DataInterface() {propertyValues {Curve {keys [
          {arriveTangent=0.4087, interpMode=RCIM_Cubic, time=0, value=0.5}
          {arriveTangent=1.4018, interpMode=RCIM_Cubic, leaveTangent=1.4018, tangentMode=RCTM_User, time=1, value=1.25}
         ]}}})))
   &DynamicMaterialParameters(
    Index 0 Param 1=&FloatFromCurve(
     FloatCurve=DataInterface() {propertyValues {Curve {defaultValue=1, keys [<at 0 0 RCIM_Cubic>, {arriveTangent=1.5072, tangentMode=RCTM_User, time=1, value=1}]}}}
     Scale Curve=0.5
    )
    Index 0 Param 2=&FloatFromCurve(
     FloatCurve=DataInterface() {propertyValues {Curve {keys [
         {arriveTangent=6.2648, interpMode=RCIM_Cubic, leaveTangent=6.2648, tangentMode=RCTM_User, time=0, value=0}
         {arriveTangentWeight=0.0536, interpMode=RCIM_Cubic, leaveTangentWeight=0.1786, tangentMode=RCTM_User, time=0.25, value=1}
         {arriveTangentWeight=0.6601, interpMode=RCIM_Cubic, leaveTangentWeight=0.088, tangentMode=RCTM_User, time=0.97, value=1}
         {arriveTangent=-14.6836, interpMode=RCIM_Cubic, leaveTangent=-14.6836, tangentMode=RCTM_User, time=1, value=0}
        ]}}}
     Scale Curve=0.6
    )
    Index 0 Param 3=&Add_Float(
     A=&FloatFromCurve(FloatCurve=DataInterface() {propertyValues {Curve {keys [{arriveTangent=0.6161, interpMode=RCIM_Cubic, time=0, value=-0.1}, {arriveTangent=0.2234, tangentMode=RCTM_User, time=0.5, value=0}]}}})
     B=&RandomRangeFloat(Minimum=0, Maximum=1)
    )
    Index 0 Param 4=5
    Module.Index0Float0Write=true
    Module.Index0Float1Write=true
    Module.Index0Float2Write=true
    Module.Index0Float3Write=true
   )
   &ScaleColor(
    Scale Alpha=&FloatFromCurve(FloatCurve=DataInterface() {propertyValues {Curve {keys [{arriveTangent=-1.84, interpMode=RCIM_Cubic, time=0, value=1}, <at 1 0>]}}})
    ScaleA=true
    ScaleRGB=true
   )
  ]
  rendererClasses [&NiagaraMeshRendererProperties]
  renderers [{rendererClass=&NiagaraMeshRendererProperties, rendererIndex=0}]
 }
}
summary {
 systemName=NS_LimeRock_RiftIn_Buildup
 emitters [
  {emitterName=FlareLight, rendererClasses [&NiagaraSpriteRendererProperties]}
  {emitterName=SmallCrack, rendererClasses [&NiagaraMeshRendererProperties]}
  {emitterName=LargeCracks, rendererClasses [&NiagaraMeshRendererProperties]}
  {emitterName=ElectricArcs, rendererClasses [&NiagaraSpriteRendererProperties]}
 ]
 userVariables [
  User.Pink {type=NiagaraBool}
  User.SizeNormalized {type=NiagaraFloat, defaultValue=4, description=""}
 ]
}
