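# Niagara one-shot burst for the rift device, apparently the exit of a rift-in.
# A fading crack, velocity-stretched curl-noise sparks, tumbling flickering shards and two timed electric-arc bursts make up the effect; the FlareLight emitter is disabled.
# User.SizeNormalized lerps counts, radii, speeds and arc size (default 4, outside 0–1), while User.Pink recolors only the arcs.
# Sparks and shards distance-cull, and sparks enlarge with LOD distance.
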NiagaraSystem /LimeRock/Assets/FX/RiftDevice/NS_LimeRock_RiftIn_Exit
dependencies {
 usedDataInterfaces [&NiagaraDataInterfaceCurve, &NiagaraDataInterfaceVector2DCurve, &NiagaraDataInterfaceColorCurve]
 usedDynamicInputs [&SimulationPosition]
 usedModules [&SystemState]
}
emitters {
 ElectricArcs {
  emitterUpdateScript [
   &EmitterState()
   &SpawnBurst_Instantaneous(Spawn Count=2, Spawn Group=1, Use Spawn Probability=false, UseLoopCountLimit=false)
   &SpawnBurst_Instantaneous(Spawn Count=3, Spawn Time=0.16, Use Spawn Probability=false, UseLoopCountLimit=false) {module=SpawnBurst_Instantaneous001}
  ]
  particleSpawnScript [
   &InitializeParticle(
    Color=&MakeCustomLinearColorFromBool(Bool=$User.Pink, True Color=<rgba 2 0.0849 1.9761 1>, False Color=<rgba 0.0813 0.6736 2 1>)
    Lifetime=&UniformRangedFloat(Minimum=0.09, Maximum=0.125)
    Mass Mode=::NewEnumerator1
    Sprite Rotation Mode=::NewEnumerator2
    Sprite Size=&Multiply_Vector2D_ByFloat(
     Float=&MaskFloatBySpawnGroup(Masked Float=1, Passthrough Float=0.5)
     Vector2D=&UniformRangedVector2D(Minimum=<200 200>, Maximum=<400 400>)
    )
    Sprite Size Mode=::NewEnumerator4
    UsePositionOffset=false
    Write Material Random=true
   ) {module=InitializeParticle001}
   &DynamicMaterialParameters(
    Index 0 Param 1=$Particles.MaterialRandom
    Index 0 Param 2=&Lerp_Float(A=0.75, B=1.2, Alpha=$Particles.MaterialRandom)
    Index 0 Param 4=1
    Module.Param0WriteEnabled=true
    Module.Param1WriteEnabled=true
    Module.Param2WriteEnabled=false
    Module.Param3WriteEnabled=true
   )
   &CameraOffset(Camera Offset Amount=20)
   &ShapeLocation(
    MaskSpawnGroup=false
    Override Seed=false
    Sphere Radius=&Lerp_Float(A=20, B=60, Alpha=$User.SizeNormalized)
    Sphere Surface Distribution=1
    UseOwnerScale=true
   )
  ]
  particleUpdateScript [
   &ParticleState()
   &DynamicMaterialParameters(
    Index 0 Param 3=&FloatFromCurve(FloatCurve=DataInterface() {propertyValues {Curve {keys [<at 0.5 0 RCIM_Cubic>, {interpMode=RCIM_Cubic, leaveTangent=3.7979, time=1, value=1}]}}})
    Index 1 Param 2=&FloatFromCurve(FloatCurve=DataInterface() {propertyValues {Curve {keys [<at 0.25 1 RCIM_Cubic>,<at 0.5 0>]}}}, CurveIndex=$System.Age, Scale Curve=0.5)
    Index 1 Param 3=&FloatFromCurve(
     FloatCurve=DataInterface() {propertyValues {Curve {keys [
         {arriveTangent=4.106, interpMode=RCIM_Cubic, leaveTangent=4.106, tangentMode=RCTM_User, time=0.05, value=0}
         {tangentMode=RCTM_User, time=0.35, value=1}
        ]}}}
     CurveIndex=$System.Age
    )
    Index 1 Param 4=100
    Module.Param0WriteEnabled=false
    Module.Param0WriteEnabled1=true
    Module.Param1WriteEnabled=false
    Module.Param1WriteEnabled1=true
    Module.Param2WriteEnabled=true
    Module.Param2WriteEnabled1=true
    Module.Param3WriteEnabled=false
    Module.Param3WriteEnabled1=true
    Write Parameter Index 1=true
   ) {module=DynamicMaterialParameters001}
   &ScaleSpriteSize(
    Uniform Curve Scale=&Lerp_Float(A=0.5, B=1, Alpha=$User.SizeNormalized)
    Uniform Curve Sprite Scale=DataInterface() {propertyValues {Curve {keys [
        {arriveTangent=0.5666, interpMode=RCIM_Cubic, leaveTangent=0.5666, tangentMode=RCTM_User, time=0, value=0.75}
        {tangentMode=RCTM_User, time=1, value=1}
       ]}}}
   )
   &ScaleColor(
    Scale RGB=&MakeVectorFromLinearColorRGB(Color=&MakeCustomLinearColorFromBool(
      Bool=$User.Pink
      True Color=&ColorFromCurve(
       ColorCurve=DataInterface() {propertyValues {
         AlphaCurve {keys []}
         BlueCurve {keys [<at 0.0806 2.7936 RCIM_Cubic>,<at 0.4355 1.3929>]}
         GreenCurve {keys [<at 0.0806 1.9309 RCIM_Cubic>,<at 0.4355 0.0957>]}
         RedCurve {keys [<at 0.0806 4 RCIM_Cubic>,<at 0.4355 1.5>]}
        }}
       CurveIndex=$Emitter.Age
      )
      False Color=&ColorFromCurve(
       ColorCurve=DataInterface() {propertyValues {
         AlphaCurve {keys []}
         BlueCurve {keys [<at 0.0806 4 RCIM_Cubic>,<at 0.4309 1.5>]}
         GreenCurve {keys [<at 0.0806 3.5603 RCIM_Cubic>,<at 0.4309 0.2162>]}
         RedCurve {keys [<at 0.0806 1.2532 RCIM_Cubic>,<at 0.4309 0.1094>]}
        }}
       CurveIndex=$Emitter.Age
      )
     ))
    ScaleA=false
    ScaleRGB=true
   ) {module=ScaleColor001}
  ]
  rendererClasses [&NiagaraSpriteRendererProperties]
  renderers [{rendererClass=&NiagaraSpriteRendererProperties}]
 }
 FlareLight {
  bEnabled=false
  emitterUpdateScript [
   &EmitterState()
   &SpawnBurst_Instantaneous(Spawn Time=0.1, Use Spawn Probability=false, UseLoopCountLimit=false)
  ]
  particleSpawnScript [
   &InitializeParticle(
    Color=<rgba 0.3252 2.6943 8 1>
    Lifetime=0.6
    Sprite Size Mode=::NewEnumerator3
    Uniform Sprite Size=&Multiply_Float(A=150, B=&Lerp_Float(A=0.3, B=1, Alpha=$User.SizeNormalized))
    UsePositionOffset=false
    Write Material Random=true
   )
   /SetVariables(Particles.DistanceLerp:NiagaraFloat=&FloatFromCurve(FloatCurve=DataInterface() {propertyValues {Curve {keys [<at 500 0>,<at 7000 1>]}}}, CurveIndex=$Engine.Owner.LODDistance)) {isSetParameters=true}
  ]
  particleUpdateScript [
   &ParticleState()
   &ScaleSpriteSize(Uniform Curve Sprite Scale=DataInterface() {propertyValues {Curve {keys [
        {arriveTangent=5.2794, interpMode=RCIM_Cubic, leaveTangent=5.2794, tangentMode=RCTM_User, time=0, value=1}
        {arriveTangent=0.7472, interpMode=RCIM_Cubic, leaveTangent=0.7472, tangentMode=RCTM_User, time=1, value=4}
       ]}}})
   &CameraOffset(Camera Offset Amount=&Multiply_Float(
     A=&Lerp_Float(A=15, B=100, Alpha=$Particles.NormalizedAge)
     B=&Lerp_Float(A=0.4, B=1, Alpha=$User.SizeNormalized)
    ))
   /SetVariables(Particles.GlowCamOffset:NiagaraFloat=&Lerp_Float(A=120, B=1200, Alpha=$Particles.DistanceLerp)) {isSetParameters=true}
   /SetVariables(Particles.DarkenColor:LinearColor=&MakeLinearColorFromVectorAndFloat(
     "Vector (RGB)"=&MakeVectorFromLinearColorRGB(Color=<rgba 0.01 0.0143 0.02 1>)
     "Float (Alpha)"=&FloatFromCurve(
      FloatCurve=DataInterface() {propertyValues {Curve {keys [{interpMode=RCIM_Cubic, tangentMode=RCTM_User, time=0, value=1}, {interpMode=RCIM_Cubic, tangentMode=RCTM_User, time=1, value=0}]}}}
      Scale Curve=1.5
     )
    )) {isSetParameters=true}
   /SetVariables(Particles.DarkenSize:Vector2f=&Multiply_Vector2D_ByFloat(Float=1.4, Vector2D=$Particles.SpriteSize)) {isSetParameters=true}
   /SetVariables(
    Particles.LightColor=&MultiplyLinearColor(
     Linear Color=$Particles.Color
     Scale Factor=&MakeFloat4FromVectorAndFloat("Vector (RGB)"=<1 1 1>, "Float (Alpha)"=20)
    )
    Particles.LightRadius=&FloatFromCurve(
     FloatCurve=DataInterface() {propertyValues {Curve {keys [
         {arriveTangent=3.5878, interpMode=RCIM_Cubic, leaveTangent=3.5878, tangentMode=RCTM_User, time=0, value=0.35}
         {arriveTangent=0.1, interpMode=RCIM_Cubic, leaveTangent=0.1, time=0.3, value=0.9847}
         {arriveTangent=5.1464, interpMode=RCIM_Cubic, leaveTangent=5.1464, tangentMode=RCTM_User, time=0.5, value=0.4}
         {arriveTangent=-0.2, arriveTangentWeight=0.0807, interpMode=RCIM_Cubic, leaveTangent=-0.2, leaveTangentWeight=0.3549, time=0.6, value=0.75}
         {arriveTangent=-0.0405, interpMode=RCIM_Cubic, leaveTangent=-0.0405, tangentMode=RCTM_User, time=1, value=0.3}
        ]}}}
     Scale Curve=&Multiply_Float(
      A=&Lerp_Float(A=150, B=400, Alpha=$User.SizeNormalized)
      B=&Lerp_Float(A=1, B=3, Alpha=$Particles.DistanceLerp)
     )
    )
   ) {isSetParameters=true}
   /SetVariables(Particles.WideFlareSize:Vector2f=&Vector2DFromCurve(
     Vector2Curve=DataInterface() {propertyValues {
       XCurve {keys [
         {arriveTangent=4.673, interpMode=RCIM_Cubic, leaveTangent=4.673, tangentMode=RCTM_User, time=0, value=0}
         {arriveTangentWeight=0.1536, interpMode=RCIM_Cubic, leaveTangentWeight=0.1152, tangentMode=RCTM_User, time=0.3, value=1}
         {arriveTangent=-3.1914, interpMode=RCIM_Cubic, leaveTangent=-3.1914, tangentMode=RCTM_User, time=0.5, value=0.15}
         {arriveTangent=-2.3891, interpMode=RCIM_Cubic, leaveTangent=-2.3891, tangentMode=RCTM_User, time=0.7, value=0.25}
         {interpMode=RCIM_Cubic, leaveTangent=-1.231, time=1, value=0}
        ]}
       YCurve {keys [
         {arriveTangent=10.6903, interpMode=RCIM_Cubic, leaveTangent=10.6903, tangentMode=RCTM_User, time=0, value=0}
         {arriveTangent=0.8, interpMode=RCIM_Cubic, leaveTangent=0.8, time=0.25, value=1}
         {arriveTangent=-1.7663, interpMode=RCIM_Cubic, leaveTangent=-1.7663, tangentMode=RCTM_User, time=0.5, value=0.4}
         {arriveTangent=-4.8806, interpMode=RCIM_Cubic, leaveTangent=-4.8806, tangentMode=RCTM_User, time=0.7, value=1}
         {interpMode=RCIM_Cubic, leaveTangent=-2.2148, time=1, value=0}
        ]}
      }}
     ScaleCurve=&Multiply_Vector2D_ByFloat(
      Float=&Lerp_Float(A=0.4, B=1, Alpha=$User.SizeNormalized)
      Vector2D=&Multiply_Vector2D_ByFloat(Float=&Lerp_Float(A=1, B=3, Alpha=$Particles.DistanceLerp), Vector2D=<1600 150>)
     )
    )) {isSetParameters=true}
   /SetVariables(Particles.GlowSize=&Vector2DFromFloat(Value=&Multiply_Float(
      A=&Lerp_Float(A=500, B=2000, Alpha=$Particles.DistanceLerp)
      B=&Lerp_Float(A=0.5, B=1, Alpha=$User.SizeNormalized)
     ))) {isSetParameters=true}
   &DynamicMaterialParameters(
    Index 0 Param 1=1
    Index 0 Param 2=&Lerp_Float(
     A=1.25
     B=0
     Alpha=&FloatFromCurve(FloatCurve=DataInterface() {propertyValues {Curve {keys [
          {arriveTangent=1.3124, interpMode=RCIM_Cubic, leaveTangent=1.3124, tangentMode=RCTM_User, time=0, value=0}
          {arriveTangent=0.3903, interpMode=RCIM_Cubic, leaveTangent=0.3903, tangentMode=RCTM_User, time=1, value=1}
         ]}}})
    )
    Index 3 Param 1=2
    Index 3 Param 2=&Lerp_Float(A=0.2, B=5, Alpha=$Particles.DistanceLerp)
    Module.Index0Param0Write=true
    Module.Index0Param1Write=true
    Module.Index0Param2Write=true
    Module.Index0Param3Write=false
    Module.Index3Param0Write=true
    Module.Index3Param1Write=true
    Module.Index3Param2Write=false
    Module.Index3Param3Write=false
    Write Parameter Index 3=true
   )
   &ScaleColor(
    Scale Alpha=&FloatFromCurve(FloatCurve=DataInterface() {propertyValues {Curve {keys [<at 0 1 RCIM_Cubic>,<at 1 0>]}}})
    ScaleA=true
    ScaleRGB=false
   )
   /SetVariables(Particles.GlowColor=&MultiplyLinearColor(
     Linear Color=$Particles.Color
     Scale Factor=&MakeFloat4FromVectorAndFloat(
      "Vector (RGB)"=<1 1 1>
      "Float (Alpha)"=&FloatFromCurve(
       FloatCurve=DataInterface() {propertyValues {Curve {keys [
           {arriveTangent=3.5878, interpMode=RCIM_Cubic, leaveTangent=3.5878, tangentMode=RCTM_User, time=0, value=0.35}
           {arriveTangent=0.3, interpMode=RCIM_Cubic, leaveTangent=0.3, time=0.3, value=1}
           {arriveTangent=3.4933, interpMode=RCIM_Cubic, leaveTangent=3.4933, tangentMode=RCTM_User, time=0.5, value=0.5}
           {arriveTangent=-0.4, arriveTangentWeight=0.0807, interpMode=RCIM_Cubic, leaveTangent=-0.4, leaveTangentWeight=0.3549, time=0.6, value=0.75}
           {arriveTangent=-0.0405, interpMode=RCIM_Cubic, leaveTangent=-0.0405, tangentMode=RCTM_User, time=1, value=0.3}
          ]}}}
       Scale Curve=0.05
      )
     )
    )) {isSetParameters=true}
  ]
  rendererClasses [&NiagaraSpriteRendererProperties, &NiagaraLightRendererProperties]
  renderers [
   {rendererClass=&NiagaraSpriteRendererProperties}
   {rendererClass=&NiagaraSpriteRendererProperties, rendererIndex=1}
   {rendererClass=&NiagaraLightRendererProperties, rendererIndex=2}
   {rendererClass=&NiagaraSpriteRendererProperties, rendererIndex=3}
   {rendererClass=&NiagaraSpriteRendererProperties, rendererIndex=4}
  ]
 }
 Shards {
  emitterUpdateScript [
   &EmitterState(
    Enable Distance Culling=true
    MaxDistance=12000
    Module.bUseMaxDistance=true
    Module.bUseMinDistance=false
    Scalability Mode=::NewEnumerator1
   )
   &SpawnBurst_Instantaneous(
    Spawn Count=&MakeIntFromFloat(Float=&Lerp_Float(A=12, B=20, Alpha=$User.SizeNormalized), Conversion Mode=::NewEnumerator1)
    Spawn Time=0.05
    Use Spawn Probability=false
    UseLoopCountLimit=false
   )
  ]
  particleSpawnScript [
   &InitializeParticle(
    Color=<rgba 2.0428 5.807 8 1>
    Lifetime Max=2.5
    Lifetime Mode=::NewEnumerator1
    Mesh Scale Max=<3 3 3>
    Mesh Scale Min=<0.7 0.7 2>
    Mesh Scale Mode=::NewEnumerator2
    Sprite Size Mode=::NewEnumerator1
    Uniform Sprite Size Max=200
    Uniform Sprite Size Min=50
    UsePositionOffset=false
    Write Material Random=true
   )
   /SetVariables(Particles.MeshIndex=&RandomRangeInt(Minimum=0, Maximum=&RandomRangeInt(Minimum=0, Maximum=1))) {isSetParameters=true}
   &ShapeLocation(
    MaskSpawnGroup=false
    Non Uniform Scale=<1 1 0.5>
    Override Seed=false
    Sphere Radius=&Lerp_Float(A=10, B=25, Alpha=$User.SizeNormalized)
    Sphere Surface Distribution=1
    UseOwnerScale=true
   )
   &AddVelocity(
    Module.Use Falloff Distance=false
    Origin Offset=<-80 0 0>
    Override Seed=false
    Velocity Mode=::NewEnumerator1
    Velocity Speed=&Multiply_Float(
     A=&RandomRangeFloat(Minimum=400, Maximum=1500)
     B=&Lerp_Float(A=0.5, B=1, Alpha=$User.SizeNormalized)
    )
   )
   &InitialMeshOrientation(Use Rotation Vector=false)
  ]
  particleUpdateScript [
   &ParticleState()
   &ScaleSpriteSize(
    Uniform Curve Index=&Add_Float(
     A=&Multiply_Float(A=$Particles.Age, B=&RandomRangeFloat(Minimum=0.8, Maximum=1.5))
     B=&RandomRangeFloat(Minimum=-0.25, Maximum=0.25)
    )
    Uniform Curve Sprite Scale=DataInterface() {propertyValues {Curve {keys [
        {tangentMode=RCTM_User, time=0.2, value=0}
        <at 0.22 1 RCIM_Cubic>
        {tangentMode=RCTM_User, time=0.5, value=0}
        <at 0.68 0>
        <at 0.7 0.8 RCIM_Cubic>
        <at 0.85 0>
       ]}}}
   )
   &UpdateMeshOrientation(
    Rotation Rate=&FloatFromCurve(
     FloatCurve=DataInterface() {propertyValues {Curve {keys [
         <at 0 1 RCIM_Cubic>
         {arriveTangent=-2.1341, interpMode=RCIM_Cubic, leaveTangent=-2.1341, tangentMode=RCTM_User, time=0.35, value=0.65}
         <at 0.8 0.1 RCIM_Cubic>
        ]}}}
     Scale Curve=&RandomRangeFloat(Minimum=2, Maximum=4)
    )
    Rotation Vector=&RandomVector(Vector Scale=1)
   )
   &Drag(
    Drag=&FloatFromCurve(
     FloatCurve=DataInterface() {propertyValues {Curve {keys [
         {arriveTangent=4.7895, interpMode=RCIM_Cubic, leaveTangent=4.7895, tangentMode=RCTM_User, time=0, value=0.1}
         {arriveTangent=0.6, interpMode=RCIM_Cubic, leaveTangent=0.6, time=0.25, value=1}
         {arriveTangent=-4.8446, interpMode=RCIM_Cubic, leaveTangent=-4.8446, tangentMode=RCTM_User, time=0.5, value=0.4}
        ]}}}
     Scale Curve=5
    )
    Use Linear Drag=true
    Use Rotational Drag=false
   )
   &SolveForcesAndVelocity(Clamp Velocity=false, Limit Acceleration=false)
   &ScaleColor(
    Scale Alpha=&FloatFromCurve(FloatCurve=DataInterface() {propertyValues {Curve {keys [<at 0.7 1 RCIM_Cubic>, {arriveTangent=-3.1986, tangentMode=RCTM_User, time=1, value=0}]}}})
    Scale RGB=&VectorFromFloat(Value=&FloatFromCurve(FloatCurve=DataInterface() {propertyValues {Curve {keys [{arriveTangent=-13.9762, interpMode=RCIM_Cubic, leaveTangent=-13.9762, tangentMode=RCTM_User, time=0, value=5}, <at 0.5 1>]}}}))
    ScaleA=true
    ScaleRGB=true
   )
  ]
  rendererClasses [&NiagaraMeshRendererProperties, &NiagaraSpriteRendererProperties]
  renderers [
   {rendererClass=&NiagaraMeshRendererProperties, rendererIndex=0}
   {rendererClass=&NiagaraSpriteRendererProperties, rendererIndex=1}
  ]
 }
 SmallCrack {
  emitterUpdateScript [
   &EmitterState()
   &SpawnBurst_Instantaneous(Use Spawn Probability=false, UseLoopCountLimit=false)
  ]
  particleSpawnScript [
   &InitializeParticle(
    Color=<rgba 2.0772 4.8329 10 1.5>
    Lifetime=0.5
    Mesh Scale Mode=::NewEnumerator3
    Mesh Uniform Scale=10
    Sprite Size Mode=::NewEnumerator3
    Uniform Sprite Size=800
    UsePositionOffset=false
    Write Material Random=true
   )
   &InitialMeshOrientation(Use Rotation Vector=false)
   &CameraOffset(Camera Offset Amount=15)
  ]
  particleUpdateScript [
   &ParticleState()
   &ScaleMeshSize(Scale Factor=&VectorFromFloat(Value=&FloatFromCurve(FloatCurve=DataInterface() {propertyValues {Curve {keys [
          {arriveTangent=0.4087, interpMode=RCIM_Cubic, leaveTangent=0.4087, tangentMode=RCTM_User, time=0, value=1}
          {arriveTangent=0.136, interpMode=RCIM_Cubic, leaveTangent=0.136, tangentMode=RCTM_User, time=1, value=1.25}
         ]}}})))
   &DynamicMaterialParameters(
    Index 0 Param 1=1
    Index 0 Param 2=&FloatFromCurve(FloatCurve=DataInterface() {propertyValues {Curve {keys [<at 0 1 RCIM_Cubic>,<at 1 0>]}}}, Scale Curve=3)
    Index 0 Param 4=5
    Module.Param0WriteEnabled=true
    Module.Param1WriteEnabled=true
    Module.Param2WriteEnabled=true
    Module.Param3WriteEnabled=true
   )
   &ScaleColor(
    Scale Alpha=&FloatFromCurve(FloatCurve=DataInterface() {propertyValues {Curve {keys [{arriveTangent=-1.84, interpMode=RCIM_Cubic, time=0, value=1}, <at 1 0>]}}})
    ScaleA=true
    ScaleRGB=true
   )
  ]
  rendererClasses [&NiagaraMeshRendererProperties, &NiagaraSpriteRendererProperties]
  renderers [
   {rendererClass=&NiagaraMeshRendererProperties, rendererIndex=0}
   {rendererClass=&NiagaraSpriteRendererProperties, rendererIndex=1}
  ]
 }
 Sparks {
  emitterUpdateScript [
   &EmitterState(
    Enable Distance Culling=true
    Max Distance Response=::KillImmediately
    MaxDistance=12000
    Module.bUseMaxDistance=true
    Module.bUseMinDistance=false
    Scalability Mode=::NewEnumerator1
   )
   &SpawnBurst_Instantaneous(
    Spawn Count=&MakeIntFromFloat(Float=&Lerp_Float(A=20, B=30, Alpha=$User.SizeNormalized), Conversion Mode=::NewEnumerator1)
    Spawn Group=1
    Use Spawn Probability=false
    UseLoopCountLimit=false
   ) {module=SpawnBurst_Instantaneous001}
  ]
  particleSpawnScript [
   /SetVariables(Particles.Distribution:NiagaraFloat=&FloatFromCurve(
     FloatCurve=DataInterface() {propertyValues {Curve {keys [
         {interpMode=RCIM_Cubic, leaveTangent=0.4364, tangentMode=RCTM_Break, time=0, value=0}
         {arriveTangent=1.9884, interpMode=RCIM_Cubic, leaveTangent=1.9884, tangentMode=RCTM_User, time=0.64, value=0.5}
         <at 1 1 RCIM_Cubic>
        ]}}}
     CurveIndex=&RandomRangeFloat(Minimum=0, Maximum=1)
    )) {isSetParameters=true}
   &InitializeParticle(
    Color=&LerpLinearColors(StartColor=<rgba 0.141 0.5013 6 1>, EndColor=<rgba 0.0331 3.0407 8 1>, LerpFactor=$Particles.Distribution)
    Lifetime=&MaskFloatBySpawnGroup(Masked Float=0.2, Passthrough Float=&RandomRangeFloat(Minimum=0.5, Maximum=1.5))
    Sprite Size Mode=::NewEnumerator3
    Uniform Sprite Size=&Lerp_Float(A=3, B=6, Alpha=$Particles.Distribution)
    UsePositionOffset=false
    Write Material Random=true
   )
   &ShapeLocation(
    MaskSpawnGroup=false
    Override Seed=false
    Sphere Radius=&MaskFloatBySpawnGroup(Masked Float=3, Passthrough Float=&Lerp_Float(A=15, B=25, Alpha=$User.SizeNormalized))
    Sphere Surface Distribution=1
    UseOwnerScale=true
   )
   &AddVelocity(
    Module.Use Falloff Distance=false
    Origin Offset=<-30 0 0>
    Override Seed=false
    Velocity Mode=::NewEnumerator1
    Velocity Speed=&Multiply_Float(
     A=&RandomRangeFloat(Minimum=900, Maximum=3000)
     B=&Lerp_Float(A=0.5, B=1, Alpha=$User.SizeNormalized)
    )
   )
  ]
  particleUpdateScript [
   &ParticleState()
   &DynamicMaterialParameters(
    Index 0 Param 1=&FloatFromCurve(FloatCurve=DataInterface() {propertyValues {Curve {keys [{arriveTangent=1.8602, interpMode=RCIM_Cubic, leaveTangent=1.8602, tangentMode=RCTM_User, time=0, value=1}, <at 1 2>]}}})
    Index 0 Param 2=5
    Index 0 Param 3=3
    Index 0 Param 4=&FloatFromCurve(FloatCurve=DataInterface() {propertyValues {Curve {keys [{arriveTangentWeight=0.1944, interpMode=RCIM_Cubic, leaveTangentWeight=0.1944, tangentMode=RCTM_User, time=0.4, value=0}, <at 0.9 1>]}}})
    Module.Param0WriteEnabled=true
    Module.Param1WriteEnabled=true
    Module.Param2WriteEnabled=true
    Module.Param3WriteEnabled=true
   )
   &Drag(
    Drag=&MaskFloatBySpawnGroup(
     Masked Float=12
     Passthrough Float=&FloatFromCurve(
      FloatCurve=DataInterface() {propertyValues {Curve {keys [
          {arriveTangent=4.106, interpMode=RCIM_Cubic, leaveTangent=4.106, tangentMode=RCTM_User, time=0, value=0.1}
          {arriveTangent=0.3, interpMode=RCIM_Cubic, leaveTangent=0.3, time=0.25, value=1}
          {arriveTangent=-6.537, interpMode=RCIM_Cubic, leaveTangent=-6.537, tangentMode=RCTM_User, time=0.5, value=0.25}
         ]}}}
      Scale Curve=9
     )
    )
    Use Linear Drag=true
    Use Rotational Drag=false
   )
   &CurlNoiseForce(
    Noise Frequency=&RandomRangeFloat(Minimum=40, Maximum=60)
    Noise Strength=&FloatFromCurve(
     FloatCurve=DataInterface() {propertyValues {Curve {keys [
         {arriveTangent=2.2956, interpMode=RCIM_Cubic, leaveTangent=2.2956, tangentMode=RCTM_User, time=0, value=0}
         {interpMode=RCIM_Cubic, tangentMode=RCTM_User, time=0.5, value=1}
         {arriveTangent=-2.5698, interpMode=RCIM_Cubic, leaveTangent=-2.5698, tangentMode=RCTM_User, time=1, value=0.4}
        ]}}}
     Scale Curve=&MaskFloatBySpawnGroup(Masked Float=0, Passthrough Float=1600)
    )
    Pan Noise=false
   )
   &SolveForcesAndVelocity(Clamp Velocity=false, Limit Acceleration=false)
   &ScaleSpriteSize(
    Uniform Curve Scale=&FloatFromCurve(FloatCurve=DataInterface() {propertyValues {Curve {keys [<at 1000 1>,<at 3000 3>]}}}, CurveIndex=$Engine.Owner.LODDistance)
    Uniform Curve Sprite Scale=DataInterface() {propertyValues {Curve {keys [<at 0.75 1 RCIM_Cubic>, {arriveTangent=-7.3774, interpMode=RCIM_Cubic, leaveTangent=-7.3774, tangentMode=RCTM_User, time=1, value=0}]}}}
   )
   &ScaleSpriteSizeBySpeed(
    Max Scale Factor=<1 10>
    Min Scale Factor=<1.2 1>
    Sample Scale Factor By Curve=true
    Scale Factor Curve=DataInterface() {propertyValues {Curve {keys [
        {arriveTangent=0.6496, interpMode=RCIM_Cubic, time=0, value=0}
        {arriveTangent=1.4663, interpMode=RCIM_Cubic, leaveTangent=1.4663, tangentMode=RCTM_User, time=1, value=1}
       ]}}}
    Velocity Threshold=2500
   )
  ]
  rendererClasses [&NiagaraSpriteRendererProperties]
  renderers [{rendererClass=&NiagaraSpriteRendererProperties}]
 }
}
summary {
 systemName=NS_LimeRock_RiftIn_Exit
 emitters [
  {
   bEnabled=false
   emitterName=FlareLight
   rendererClasses [&NiagaraSpriteRendererProperties, &NiagaraLightRendererProperties]
  }
  {emitterName=SmallCrack, rendererClasses [&NiagaraMeshRendererProperties, &NiagaraSpriteRendererProperties]}
  {emitterName=Sparks, rendererClasses [&NiagaraSpriteRendererProperties]}
  {emitterName=ElectricArcs, rendererClasses [&NiagaraSpriteRendererProperties]}
  {emitterName=Shards, rendererClasses [&NiagaraMeshRendererProperties, &NiagaraSpriteRendererProperties]}
 ]
 userVariables [
  User.Pink {type=NiagaraBool}
  User.SizeNormalized {type=NiagaraFloat, defaultValue=4, description=""}
 ]
}
